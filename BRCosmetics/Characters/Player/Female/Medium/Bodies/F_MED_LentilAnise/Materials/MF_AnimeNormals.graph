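# Material function in the LentilAnise character folder that reshapes the incoming normal for anime-style toon shading.
# Vertex-color-encoded normals blend in by a parameter times vertex alpha.
# Sphere-to-cylinder normals come from pre-skinned local position around a center parameter, with a smoothstep height fade, optionally multiplied by another mask.
# Static switches select the output, optionally blending in a baked capsule tangent normal map.
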MaterialFunction /BRCosmetics/Characters/Player/Female/Medium/Bodies/F_MED_LentilAnise/Materials/MF_AnimeNormals
graph {nodes [
  n1 = TextureSampleParameter2D() {
   group=Toon_Shading
   parameterName=BakedCapsuleTangentNormal
   samplerType=SAMPLERTYPE_Normal
   texture=&FlatNormalBlue
  }
  n2 = LinearInterpolate(A=NamedRerouteUsage(), B=Transform(None=@n1.RGB), Alpha=NamedRerouteUsage())
  n3 = NamedRerouteUsage()
  FunctionOutput(None=StaticSwitchParameter(True=StaticSwitchParameter(True=@n2, False=@n3) {group=Toon_Shading, parameterName=UseCapsuleTangentNormal}, False=NamedRerouteUsage()) {group=Toon_Shading, parameterName=UseSphereCylindricalNormal}) {outputName=Out_Normal}
  NamedRerouteDeclaration(None=FunctionInput() {inputName=In_Normal}) {name=In_Normal, nodeColor=<rgba 0.7344 0 1 1>}
  n4 = LocalPosition() {localOrigin=InstancePreSkinning}
  NamedRerouteDeclaration(None=@n4.Z) {name=PreSkinnedZ, nodeColor=<rgba 0.4666 0.0287 1 1>}
  n5 = NamedRerouteUsage()
  n6 = SmoothStep(
   Min=ScalarParameter() {defaultValue=140, group=Toon_Shading, parameterName=NormalSphereFadeStart}
   Max=ScalarParameter() {defaultValue=150, group=Toon_Shading, parameterName=NormalSphereFadeEnd}
   Value=VertexInterpolator(VS=NamedRerouteUsage()).PS
  ) {constMax=60, constMin=45}
  NamedRerouteDeclaration(None=StaticSwitchParameter(True=Multiply(A=@n5, B=@n6), False=@n6) {group=Toon_Shading, parameterName=UseMultiplyVertAlphaForFade}) {name=FadeStartEnd, nodeColor=<rgba 1 0.8906 0 1>}
  n7 = NamedRerouteUsage()
  n8 = VertexColor()
  n9 = Transform(None=Subtract(A=Multiply(A=@n8) {constB=2})) {transformSourceType=TRANSFORMSOURCE_Local}
  n10 = ScalarParameter() {desc="0.0 = surface normals, 1.0 baked vertex color normal", group=02_Anime Parameters, parameterName=VertexBakedNormal_Blend, sliderMax=1}
  n11 = Multiply(A=@n10, B=NamedRerouteDeclaration(None=@n8.A) {name=VertexColorAlpha, nodeColor=<rgba 0.7344 1 0 1>})
  NamedRerouteDeclaration(None=Normalize(VectorInput=LinearInterpolate(A=@n7, B=@n9, Alpha=@n11))) {name=VertexNormalFinal, nodeColor=<rgba 0 1 0.7188 1>}
  n12 = NamedRerouteDeclaration(None=Transform(None=NamedRerouteUsage())) {name=NormalZWorldSpace, nodeColor=<rgba 0.0765 0.1715 1 1>}
  n13 = Subtract(A=@n4.XYZ, B=VectorParameter() {group=Toon_Shading, parameterName=NormalSphere_Location, defaultValue=<rgba 0 0 160 1>}.RGB)
  n14 = LinearInterpolate(
   A=@n13
   B=Multiply(A=@n13, B=Constant3Vector() {constant=<rgba 1 1 0 0>})
   Alpha=ScalarParameter() {defaultValue=1, group=Toon_Shading, parameterName=NormalSphereCylindrical_Blend, sliderMax=1}
  )
  n15 = VertexInterpolator(VS=Transform(None=Normalize(VectorInput=@n14)) {transformSourceType=TRANSFORMSOURCE_Local})
  n16 = ScalarParameter() {defaultValue=0.75, group=Toon_Shading, parameterName=NormalSphere_Blend, sliderMax=1}
  NamedRerouteDeclaration(None=LinearInterpolate(A=@n12, B=LinearInterpolate(A=@n12, B=@n15.PS, Alpha=@n16) {constAlpha=0}, Alpha=NamedRerouteUsage())) {name=SphereNormalFinal, nodeColor=<rgba 0.4766 0 1 1>}
 ]}
