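# Material function for the StairGummy character that outputs a DFL mask for a view-dependent anime nose line.
# Camera direction against head-bone right and up vectors drives a flip step and a vertical fade with separate up/down thresholds, shaped by a pre-skinned height sphere mask.
# The fade multiplies DFL_Map red, then adds blue, behind the UseDFLMap switch.
# Both flip inputs read red, so flipping changes nothing.
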MaterialFunction /BRCosmetics/Characters/Player/Female/Medium/Bodies/F_MED_StairGummy/Materials/MF_NoseLineFlip_StairGummy
graph {nodes [
  n1 = TextureSampleParameter2D() {
   group=01_Anime Maps
   parameterName=DFL_Map
   samplerType=SAMPLERTYPE_Masks
   sortPriority=14
   texture=&T_BlackMask
  }
  n2 = VectorParameter() {group=Nose, parameterName=HeadBone_VecRight, defaultValue=<rgba 1 0 0 1>}
  n3 = VectorParameter() {group=Nose, parameterName=HeadBone_Location, defaultValue=<rgba 0 0 160 1>}
  n4 = TransformPosition(None=@n3.RGB) {transformType=TRANSFORMPOSSOURCE_World}
  n5 = NamedRerouteUsage()
  n6 = Normalize(VectorInput=Subtract(A=CameraPositionWS(), B=LinearInterpolate(A=@n4, B=@n3.RGB, Alpha=@n5)))
  n7 = DotProduct(A=LinearInterpolate(A=Transform(None=@n2.RGB) {transformSourceType=TRANSFORMSOURCE_Local}, B=@n2.RGB, Alpha=NamedRerouteUsage()), B=@n6)
  n8 = LinearInterpolate(A=@n1.R, B=@n1.R, Alpha=Step(Y=Saturate(None=@n7), X=ScalarParameter() {group=Nose, parameterName=NoseFlip_Threshold}) {constX=0})
  n9 = VertexInterpolator(VS=LocalPosition() {localOrigin=InstancePreSkinning}.XYZ)
  n10 = ScalarParameter() {defaultValue=0.4, group=Nose, parameterName=NoseHide_ThresholdUp}
  n11 = VectorParameter() {group=Nose, parameterName=HeadBone_VecUp, defaultValue=<rgba 0 0 1 1>}
  n12 = DotProduct(A=@n6, B=LinearInterpolate(A=Transform(None=@n11.RGB) {transformSourceType=TRANSFORMSOURCE_Local}, B=@n11.RGB, Alpha=NamedRerouteUsage()))
  n13 = Abs(None=@n12)
  n14 = ScalarParameter() {defaultValue=0.4, group=Nose, parameterName=NoseHide_ThresholdDown}
  n15 = Sign(None=@n12)
  n16 = MaterialFunctionCall!(
   1=Constant()
   2=SphereMask(
    A=ComponentMask(None=@n9.PS) {b=true}
    B=ScalarParameter() {defaultValue=162, group=Nose, parameterName=Nose_FadeHeight}
    Radius=ScalarParameter() {defaultValue=5, group=Nose, parameterName=Nose_FadeRadius}
   ) {attenuationRadius=5, hardnessPercent=0}
   3=Constant() {r=1}
   4
   A=LinearInterpolate(A=SmoothStep(Min=@n10, Value=@n13) {constMax=0}, B=SmoothStep(Min=@n14, Value=@n13) {constMax=0}, Alpha=Saturate(None=@n15))
  ) {materialFunction=/Engine/Functions/Engine_MaterialFunctions02/Lerp_Multiple_Float}
  n17 = Saturate(None=Add(A=@n1.B, B=Multiply(A=@n8, B=@n16.Lerp 3 Inputs)))
  n18 = Constant()
  FunctionOutput(None=NamedRerouteDeclaration(None=StaticSwitchParameter(True=@n17, False=@n18) {group=01_Anime Maps, parameterName=UseDFLMap, sortPriority=4}) {name=DFL, nodeColor=<rgba 1 0.9297 0 1>}) {outputName=DFL}
  NamedRerouteDeclaration(None=@n16.Lerp 3 Inputs) {name=Debug, nodeColor=<rgba 0 1 0.6016 1>}
  NamedRerouteDeclaration(None=ScalarParameter() {group=Nose, parameterName=Nose_UseWS}) {name=_useWS, nodeColor=<rgba 0 1 0.1094 1>}
 ]}
