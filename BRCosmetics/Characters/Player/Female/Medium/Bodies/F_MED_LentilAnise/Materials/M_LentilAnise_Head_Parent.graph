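# Parent material for the LentilAnise character head, an anime toon shader on the Cloth shading model.
# It chains shared 2023 anime Pre, Calculations and Post functions, inserting eye animation on color, masked by texture, vertex color or UV1.
# Borrowed StairGummy and SilentNovel functions handle nose-line flipping and eyeball masking of DF lines.
# Adjustable spherical-to-cylindrical normals, computed per vertex from pre-skinned local position, blend with mapped normals to smooth toon shading.
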Material /BRCosmetics/Characters/Player/Female/Medium/Bodies/F_MED_LentilAnise/Materials/M_LentilAnise_Head_Parent
graph {
 properties {
  bAutomaticallySetUsageInEditor=false
  bHasPixelAnimation=true
  bUseMaterialAttributes=true
  bUsedWithClothing=true
  bUsedWithMeshDeformer=true
  bUsedWithMorphTargets=true
  bUsedWithNiagaraMeshParticles=true
  bUsedWithSkeletalMesh=true
  bUsedWithStaticLighting=true
  pixelDepthOffsetMode=PDOM_Legacy
  shadingModel=MSM_Cloth
  usedShadingModels=Cloth
 }
 nodes [
  n1 = MaterialFunctionCall!(CustomUV) {materialFunction=/Game/Global/Materials/Character_Parent_2023/Anime_Parent_2023/MaterialFunctions/MFC_AnimeShader_2023_Pre}
  n2 = MaterialFunctionCall!(
   CustomLightVector
   OverrideNormal_WS=NamedRerouteUsage()
   STT=@n1.STT
   Normal_Map=NamedRerouteUsage()
   DFL=MaterialFunctionCall() {materialFunction=/BRCosmetics/Characters/Player/Female/Medium/Bodies/F_MED_StairGummy/Materials/MF_NoseLineFlip_StairGummy}.DFL
   FX=@n1.STT
  ) {materialFunction=/Game/Global/Materials/Character_Parent_2023/Anime_Parent_2023/MaterialFunctions/MFC_AnimeShader_2023_Calculations}
  n3 = VertexColor()
  n4 = TextureSampleParameter2D() {
   group=01_EyeAnimation_Parameters
   parameterName=EyeMask
   samplerSource=SSM_Wrap_WorldGroupSettings
   samplerType=SAMPLERTYPE_Masks
   sortPriority=-32
   texture=&T_WhiteMask
  }
  n5 = StaticSwitchParameter(True=AppendVector(A=@n3, B=@n3), False=@n4.RGBA) {group=01_EyeAnimation_Parameters, parameterName=UseVertexColorForEyeMask, sortPriority=-33}
  n6 = TextureCoordinate() {coordinateIndex=1}
  n7 = Sign(None=Subtract(A=ComponentMask(None=@n6) {r=true}) {constB=0.0001})
  n8 = MaterialFunctionCall!(
   Color Lit=@n1.Color_Lit
   Color Shaded=@n1.Color_Lit
   Use UV2=StaticBoolParameter() {defaultValue=true, group=01_EyeAnimation_Parameters, parameterName=Use UV2 for Eye Animation, sortPriority=-30}
   Mask=StaticSwitchParameter(
    True=ChannelMaskParameter(None=@n5) {group=01_EyeAnimation_Parameters, parameterName=EyeMaskChannel, sortPriority=-32}.R
    False=Saturate(None=@n7)
   ) {group=01_EyeAnimation_Parameters, parameterName=UseEyeMask, sortPriority=-33}
  ) {materialFunction=/Game/Global/Materials/Character_Parent_2023/Anime_Parent_2023/MaterialFunctions/MF_AnimeEyeAnimation_v2}
  MaterialFunctionCall!(
   Shading_Gradient=@n2.Shading_Gradient
   Threshold_Gradient=@n2.Shading_Gradient
   Specular_Gradient=@n2.Shading_Gradient
   Sheen_Mask=@n2.Shading_Gradient
   Thickness=@n2.Shading_Gradient
   DFL_Mask=MaterialFunctionCall!(DFLines=@n2.DFL_Mask) {materialFunction=/BRCosmetics/Characters/Player/Male/Medium/Bodies/M_MED_SilentNovel_Vane/Materials/MF_EyeballsDFLineMasking}.Result
   FX=@n2.Shading_Gradient
   Metallic_Gradient=@n2.Shading_Gradient
   RimLight_Gradient=@n2.Shading_Gradient
   NormalBuffer=@n2.Shading_Gradient
   --------
   CINE_LightSources=@n2.Shading_Gradient
   CINE_LightColor=@n2.Shading_Gradient
   ---------
   Color_Lit=@n8.Lit
   Color_Shaded=@n8.Lit
   SRM=@n1.SRM
   Emissive=@n1.SRM
   DFL_Color=@n1.SRM
   Specular_Color=@n1.SRM
   RimLight_Color=@n1.SRM
  ) {materialFunction=/Game/Global/Materials/Character_Parent_2023/Anime_Parent_2023/MaterialFunctions/MFC_AnimeShader_2023_Post}
  MaterialFunctionCall!(In_Normal=NamedRerouteUsage()) {materialFunction=&/MF_AnimeNormals}
  NamedRerouteDeclaration(None=@n1.Normal) {name=In_Normal, nodeColor=<rgba 1 0 0.75 1>}
  n9 = NamedRerouteUsage()
  n10 = Subtract(
   A=LocalPosition() {localOrigin=InstancePreSkinning}.XYZ
   B=VectorParameter() {group=Toon_Shading, parameterName=NormalSphere_Location, defaultValue=<rgba 0 0 162 1>}.RGB
  )
  n11 = Multiply(A=@n10, B=Constant3Vector() {constant=<rgba 1 1 0 0>})
  n12 = ScalarParameter() {group=Toon_Shading, parameterName=NormalSphereCylindrical_Blend, sliderMax=1}
  n13 = Transform(None=Normalize(VectorInput=LinearInterpolate(A=@n10, B=@n11, Alpha=@n12))) {transformSourceType=TRANSFORMSOURCE_Local}
  NamedRerouteDeclaration(None=LinearInterpolate(
    A=Transform(None=@n9)
    B=VertexInterpolator(VS=@n13).PS
    Alpha=ScalarParameter() {group=Toon_Shading, parameterName=NormalSphere_Blend, sliderMax=1}
   ) {constAlpha=0}) {name=BlendedNormals, nodeColor=<rgba 0 0.4072 0.1683 1>}
 ]
}
