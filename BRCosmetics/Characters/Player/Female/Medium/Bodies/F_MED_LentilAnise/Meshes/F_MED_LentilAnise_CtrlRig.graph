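# Control Rig for the LentilAnise female body cosmetic, correcting deformation of shoulders, skirt, collar, belts, chains and holstered gun.
# After a shared DeformRig function, a sequence applies weighted parent constraints, then shoulder correctives switched by Skydive and Emotes.
# Later steps use pose readers, thigh rotation and distance to offset belts, chains and left front skirt, plus aim constraints.
# Cylindrical clamps bound chain bones around the left thigh; some leftover nodes are unwired.
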ControlRigBlueprint /BRCosmetics/Characters/Player/Female/Medium/Bodies/F_MED_LentilAnise/Meshes/F_MED_LentilAnise_CtrlRig
graphs [
 RigVMFunctionLibrary {links [], nodes []}
 RigVMFunction_Sequence_1_ContainedGraph {nodes [
   n1 = RigVMFunctionEntryNode()
   RigVMFunction_Sequence = &RigVMFunction_Sequence(ExecuteContext=@n1.ExecuteContext)
   n2 = &RigVMFunction_Sequence(ExecuteContext=@RigVMFunction_Sequence.B)
   n3 = &RigVMFunction_Sequence(ExecuteContext=@n2.B)
   RigVMFunctionReturnNode(A=@RigVMFunction_Sequence.A, B=@n2.A, C=@n3.A, D=@n3.B)
  ]}
 RigVMModel {nodes [
   RigUnit_BeginExecution = &RigUnit_BeginExecution()
   FN_DeformRig_FNC = RigVMFunctionReferenceNode(
    DeformRig_ Curve="(deformRig_l_upperarm_up_alpha_72_#1=False,deformRig_r_upperarm_up_alpha_73_#2=False,deformRig_l_upperarm_fr_alpha_74_#3=False,deformRig_r_upperarm_fr_alpha_75_CB51B8 ...(truncated)"
    DeformRig_Bool="(upperarm_1_#4=True,elbow_3_#5=True,wrist_7_#6=True,groin_8_#7=True,glute_17_#8=True,knee_9_40C408324CA261 ...(truncated)"
    DeformRig_Debug=(upperarm_debug_18_#4=False,wrist_debug_20_#6=False,lat_debug_26_#9=False,groinGlute_debug_27_#7=False)
    DeformRig_Value="(upperarm_up_18_#10=(Rotation=(X=0,Y=0,Z=0,W=1),Translation=(X=0,Y=0,Z=8),Scale3D=(X=0.7,Y=1,Z=1)),upperarm_fwd_fr_19_#11=(Rotation=(X=0,Y=0,Z=0,W=1),Translation=(X=0,Y=10,Z=0),Scale3D=(X=0.5,Y=1,Z= ...(truncated)"
    ExecuteContext=@RigUnit_BeginExecution.ExecutePin
   )
   > RigVMFunction_Sequence = RigVMAggregateNode()
   &RigUnit_ParentConstraint(
    AdvancedSettings=(RotationOrderForFilter=XZY)
    Child=(Type=Bone,Name="deform_skirt_f_l")
    Parents="((Item=(Type=Bone,Name=\"upperarm_twist_01_l\"),Weight=0.0012),(Item=(Type=Bone,Name=\"spine_02\"),Weight=0.0031),(Item=(Type=Bone,Name=\"spine_03\"),Weight=0.0032),(Item=(Type=Bone,Name=\"clavicle_l\"),Weight=0.0256),(Item=(Type=Bone,Name=\"spine_05\"),Weight=0.097 ...(truncated)"
    Weight=1
    ExecutePin=@RigVMFunction_Sequence.A
   )
   > &RigUnit_ParentConstraint(
    AdvancedSettings=(RotationOrderForFilter=XZY)
    Child=(Type=Bone,Name="deform_skirt_f_m_l")
    Parents="((Item=(Type=Bone,Name=\"deform_upperarm_fr_l\"),Weight=0.0152),(Item=(Type=Bone,Name=\"spine_03\"),Weight=0.0177),(Item=(Type=Bone,Name=\"spine_02\"),Weight=0.0195),(Item=(Type=Bone,Name=\"upperarm_twist_01_l\"),Weight=0.0367),(Item=(Type=Bone,Name=\"clavicle_l\"), ...(truncated)"
    Weight=1
   )
   > &RigUnit_ParentConstraint(
    AdvancedSettings=(RotationOrderForFilter=XZY)
    Child=(Type=Bone,Name="deform_skirt_f_m_r")
    Parents="((Item=(Type=Bone,Name=\"spine_03\"),Weight=0.0141),(Item=(Type=Bone,Name=\"spine_02\"),Weight=0.0157),(Item=(Type=Bone,Name=\"deform_upperarm_fr_r\"),Weight=0.0192),(Item=(Type=Bone,Name=\"upperarm_twist_01_r\"),Weight=0.0358),(Item=(Type=Bone,Name=\"clavicle_r\"), ...(truncated)"
    Weight=1
   )
   > &RigUnit_ParentConstraint(
    AdvancedSettings=(RotationOrderForFilter=XZY)
    Child=(Type=Bone,Name="deform_skirt_f_r")
    Parents="((Item=(Type=Bone,Name=\"deform_upperarm_fr_r\"),Weight=0.0023),(Item=(Type=Bone,Name=\"upperarm_twist_01_r\"),Weight=0.0064),(Item=(Type=Bone,Name=\"spine_03\"),Weight=0.0105),(Item=(Type=Bone,Name=\"spine_02\"),Weight=0.011),(Item=(Type=Bone,Name=\"clavicle_r\"),W ...(truncated)"
    Weight=1
   )
   > &RigUnit_ParentConstraint(
    AdvancedSettings=(RotationOrderForFilter=XZY)
    Child=(Type=Bone,Name="deform_skirt_m_l")
    Parents="((Item=(Type=Bone,Name=\"clavicle_l\"),Weight=0.0057),(Item=(Type=Bone,Name=\"spine_05\"),Weight=0.0061),(Item=(Type=Bone,Name=\"spine_01\"),Weight=0.0639),(Item=(Type=Bone,Name=\"pelvis\"),Weight=0.0727),(Item=(Type=Bone,Name=\"spine_03\"),Weight=0.1274),(Item=(Typ ...(truncated)"
    Weight=1
   )
   > &RigUnit_ParentConstraint(
    AdvancedSettings=(RotationOrderForFilter=XZY)
    Child=(Type=Bone,Name="deform_skirt_m_r")
    Parents="((Item=(Type=Bone,Name=\"clavicle_r\"),Weight=0.0057),(Item=(Type=Bone,Name=\"spine_05\"),Weight=0.0061),(Item=(Type=Bone,Name=\"spine_01\"),Weight=0.0639),(Item=(Type=Bone,Name=\"pelvis\"),Weight=0.0727),(Item=(Type=Bone,Name=\"spine_03\"),Weight=0.1274),(Item=(Typ ...(truncated)"
    Weight=1
   )
   > &RigUnit_ParentConstraint(
    AdvancedSettings=(RotationOrderForFilter=XZY)
    Child=(Type=Bone,Name="deform_skirt_b_l")
    Parents="((Item=(Type=Bone,Name=\"dyn_skirt_bk_l\"),Weight=0.0326),(Item=(Type=Bone,Name=\"spine_03\"),Weight=0.0432),(Item=(Type=Bone,Name=\"spine_01\"),Weight=0.056),(Item=(Type=Bone,Name=\"dyn_skirt_b_l\"),Weight=0.0786),(Item=(Type=Bone,Name=\"spine_04\"),Weight=0.0792), ...(truncated)"
    Weight=1
   )
   > &RigUnit_ParentConstraint(
    AdvancedSettings=(RotationOrderForFilter=XZY)
    Child=(Type=Bone,Name="deform_skirt_bk_l")
    Parents="((Item=(Type=Bone,Name=\"dyn_skirt_b_l\"),Weight=0.0099),(Item=(Type=Bone,Name=\"spine_05\"),Weight=0.0107),(Item=(Type=Bone,Name=\"spine_03\"),Weight=0.0715),(Item=(Type=Bone,Name=\"spine_01\"),Weight=0.0742),(Item=(Type=Bone,Name=\"dyn_skirt_bk_l\"),Weight=0.0886) ...(truncated)"
    Weight=1
   )
   > &RigUnit_ParentConstraint(
    AdvancedSettings=(RotationOrderForFilter=XZY)
    Child=(Type=Bone,Name="deform_skirt_bk_r")
    Parents="((Item=(Type=Bone,Name=\"spine_05\"),Weight=0.0112),(Item=(Type=Bone,Name=\"dyn_skirt_bk_l\"),Weight=0.0117),(Item=(Type=Bone,Name=\"spine_03\"),Weight=0.0707),(Item=(Type=Bone,Name=\"spine_01\"),Weight=0.0725),(Item=(Type=Bone,Name=\"dyn_skirt_bk_r\"),Weight=0.0871 ...(truncated)"
    Weight=1
   )
   > &RigUnit_ParentConstraint(
    AdvancedSettings=(RotationOrderForFilter=XZY)
    Child=(Type=Bone,Name="deform_skirt_b_r")
    Parents=((Item=(Type=Bone,Name="spine_03"),Weight=0.0163),(Item=(Type=Bone,Name="spine_04"),Weight=0.0192),(Item=(Type=Bone,Name="spine_01"),Weight=0.0649),(Item=(Type=Bone,Name="spine_02"),Weight=0.2039),(Item=(Type=Bone,Name="pelvis"),Weight=0.6958))
    Weight=1
   )
   > &RigUnit_ParentConstraint(
    AdvancedSettings=(RotationOrderForFilter=XZY)
    Child=(Type=Bone,Name="deform_chain_bk")
    Parents=((Item=(Type=Bone,Name="thigh_twist_01_l"),Weight=0.0239),(Item=(Type=Bone,Name="deform_glute_l"),Weight=0.257),(Item=(Type=Bone,Name="pelvis"),Weight=0.7192))
    Weight=1
   )
   > &RigUnit_ParentConstraint(
    AdvancedSettings=(RotationOrderForFilter=XZY)
    Child=(Type=Bone,Name="deform_belt_gun_r")
    Parents="((Item=(Type=Bone,Name=\"thigh_r\"),Weight=0.023),(Item=(Type=Bone,Name=\"deform_glute_r\"),Weight=0.043),(Item=(Type=Bone,Name=\"deform_groin_r\"),Weight=0.0478),(Item=(Type=Bone,Name=\"pelvis\"),Weight=0.4),(Item=(Type=Bone,Name=\"thigh_twist_01_r\"),Weight=0.4861 ...(truncated)"
    Weight=1
   )
   > &RigUnit_ParentConstraint(
    AdvancedSettings=(RotationOrderForFilter=XZY)
    Child=(Type=Bone,Name="deform_belt_f")
    Parents="((Item=(Type=Bone,Name=\"spine_02\"),Weight=0.0234),(Item=(Type=Bone,Name=\"deform_groin_r\"),Weight=0.028),(Item=(Type=Bone,Name=\"deform_groin_l\"),Weight=0.0318),(Item=(Type=Bone,Name=\"thigh_twist_01_r\"),Weight=0.0384),(Item=(Type=Bone,Name=\"thigh_twist_01_l\" ...(truncated)"
    Weight=1
   )
   > &RigUnit_ParentConstraint(
    AdvancedSettings=(RotationOrderForFilter=XZY)
    Child=(Type=Bone,Name="deform_belt_r")
    Parents="((Item=(Type=Bone,Name=\"thigh_twist_01_l\"),Weight=0.0004),(Item=(Type=Bone,Name=\"thigh_r\"),Weight=0.0017),(Item=(Type=Bone,Name=\"thigh_twist_01_r\"),Weight=0.2186),(Item=(Type=Bone,Name=\"deform_groin_r\"),Weight=0.2867),(Item=(Type=Bone,Name=\"pelvis\"),Weight ...(truncated)"
    Weight=1
   )
   > &RigUnit_ParentConstraint(
    AdvancedSettings=(RotationOrderForFilter=XZY)
    Child=(Type=Bone,Name="deform_belt_l")
    Parents=((Item=(Type=Bone,Name="spine_02"),Weight=0.0054),(Item=(Type=Bone,Name="thigh_twist_01_l"),Weight=0.1376),(Item=(Type=Bone,Name="deform_groin_l"),Weight=0.2351),(Item=(Type=Bone,Name="pelvis"),Weight=0.6219))
    Weight=1
   )
   > &RigUnit_ParentConstraint(
    AdvancedSettings=(RotationOrderForFilter=XZY)
    Child=(Type=Bone,Name="deform_collar_l")
    Parents="((Item=(Type=Bone,Name=\"upperarm_twist_01_l\"),Weight=0.0074),(Item=(Type=Bone,Name=\"head\"),Weight=0.0133),(Item=(Type=Bone,Name=\"neck_02\"),Weight=0.0149),(Item=(Type=Bone,Name=\"deform_pec_l\"),Weight=0.0234),(Item=(Type=Bone,Name=\"dyn_collar_l\"),Weight=0.06 ...(truncated)"
    Weight=1
   )
   > &RigUnit_ParentConstraint(
    AdvancedSettings=(RotationOrderForFilter=XZY)
    Child=(Type=Bone,Name="deform_collar_r")
    Parents="((Item=(Type=Bone,Name=\"deform_upperarm_fr_r\"),Weight=0.0051),(Item=(Type=Bone,Name=\"upperarm_twist_01_r\"),Weight=0.0058),(Item=(Type=Bone,Name=\"deform_pec_r\"),Weight=0.0192),(Item=(Type=Bone,Name=\"head\"),Weight=0.0423),(Item=(Type=Bone,Name=\"dyn_collar_r\" ...(truncated)"
    Weight=1
   )
   RigVMFunction_ControlFlowBranch = &RigVMFunction_ControlFlowBranch(ExecuteContext=@RigVMFunction_Sequence.B, Condition=RigVMVariableNode() {variable=Skydive}.Value)
   n1 = &RigVMFunction_ControlFlowBranch(ExecuteContext=@RigVMFunction_ControlFlowBranch.False, Condition=RigVMVariableNode() {variable=Emotes}.Value)
   n2 = /Script/RigVM.RigVMFunction_MathQuaternionGetAxis(Axis=X, Quaternion=&RigUnit_GetTransform(Item=(Type=Bone,Name="spine_05"), bInitial=False).Transform.Rotation) {is=struct RigVMUnitNode, template="Axis::Execute(in Quaternion,in Axis,out Result)"}
   Dot___5 = /Script/RigVM.RigVMFunction_MathVectorDot(
    A=@n2.Result
    B=&RigVMFunction_MathQuaternionGetAxis(Axis=X, Quaternion=&RigUnit_GetTransform(Item=(Type=Bone,Name="upperarm_r"), bInitial=False).Transform.Rotation) {template="Axis::Execute(in Quaternion,in Axis,out Result)"}.Result
   ) {is=struct RigVMUnitNode, template="Dot,|::Execute(in A,in B,out Result)"}
   &RigUnit_ModifyTransforms(
    ItemToModify=((Item=(Type=Bone,Name="deform_upperarm_up_r"),Transform=(Rotation=(X=0,Y=0,Z=0,W=1),Translation=(X=-5,Y=0,Z=5),Scale3D=(X=1,Y=1,Z=1))))
    WeightMinimum=0
    Weight=&RigVMFunction_MathDoubleRemap(SourceMaximum=-1, SourceMinimum=-0.53, TargetMaximum=1, TargetMinimum=0, Value=@Dot___5.Result).Result
    ExecutePin=@n1.False
   )
   Dot___6 = &RigVMFunction_MathVectorDot(
    B=&RigVMFunction_MathQuaternionGetAxis(Axis=X, Quaternion=&RigUnit_GetTransform(Item=(Type=Bone,Name="upperarm_l"), bInitial=False).Transform.Rotation) {template="Axis::Execute(in Quaternion,in Axis,out Result)"}.Result
    A=@n2.Result
   ) {template="Dot,|::Execute(in A,in B,out Result)"}
   > &RigUnit_ModifyTransforms(
    ItemToModify=((Item=(Type=Bone,Name="deform_upperarm_up_l"),Transform=(Rotation=(X=0,Y=0,Z=0,W=1),Translation=(X=5,Y=0,Z=-5),Scale3D=(X=1,Y=1,Z=1))))
    WeightMinimum=0
    Weight=&RigVMFunction_MathDoubleRemap(SourceMaximum=1, SourceMinimum=0.53, TargetMaximum=1, TargetMinimum=0, Value=@Dot___6.Result).Result
   )
   n3 = &RigUnit_SphericalPoseReader(
    ActiveRegionScaleFactors=(PositiveWidth=2.5,NegativeWidth=3,PositiveHeight=3,NegativeHeight=3)
    ActiveRegionSize=0.1
    Debug=(bDrawDebug=false,DebugScale=25,DebugSegments=20,DebugThickness=0.25)
    DriverAxis=(X=1,Y=0,Z=0)
    DriverItem=(Type=Bone,Name="upperarm_r")
    FalloffRegionScaleFactors=(PositiveWidth=1,NegativeWidth=1,PositiveHeight=1,NegativeHeight=1)
    FalloffSize=0.7
    OptionalParentItem=(Type=Bone)
    RotationOffset=(X=0,Y=0,Z=0)
    ExecutePin=@RigVMFunction_ControlFlowBranch.True
   )
   > &RigUnit_ModifyTransforms(
    ItemToModify=((Item=(Type=Bone,Name="deform_upperarm_up_r"),Transform=(Rotation=(X=0,Y=0,Z=0,W=1),Translation=(X=0,Y=2,Z=4),Scale3D=(X=1,Y=1,Z=1))))
    Weight=@n3.OutputParam
   )
   > n4 = &RigUnit_SphericalPoseReader(
    ActiveRegionScaleFactors=(PositiveWidth=2.5,NegativeWidth=3,PositiveHeight=3,NegativeHeight=3)
    ActiveRegionSize=0.1
    Debug=(bDrawDebug=false,DebugScale=25,DebugSegments=20,DebugThickness=0.25)
    DriverAxis=(X=1,Y=0,Z=0)
    DriverItem=(Type=Bone,Name="upperarm_l")
    FalloffRegionScaleFactors=(PositiveWidth=1,NegativeWidth=1,PositiveHeight=1,NegativeHeight=1)
    FalloffSize=0.7
    OptionalParentItem=(Type=Bone)
    RotationOffset=(X=0,Y=0,Z=0)
   )
   > &RigUnit_ModifyTransforms(
    ItemToModify="((Item=(Type=Bone,Name=\"deform_upperarm_up_l\"),Transform=(Rotation=(X=0,Y=0.4226,Z=0,W=-0.9063),Translation=(X=10,Y=-1,Z=-8),Scale3D=(X=1,Y=1,Z=0))),(Item=(Type=Bone,Name=\"deform_upperarm_up_r\"),Transform=(Rotation=(X=0,Y=-0.4235,Z=0,W=0.9059),Translation= ...(truncated)"
    Weight=@n4.OutputParam
   )
   n5 = &RigVMFunction_Sequence(ExecuteContext=@RigVMFunction_Sequence.C)
   n6 = &RigUnit_SphericalPoseReader(
    ActiveRegionScaleFactors=(PositiveWidth=0.4,NegativeWidth=0.7,PositiveHeight=0.3,NegativeHeight=0.35)
    ActiveRegionSize=0.3
    Debug=(bDrawDebug=false,DebugScale=25,DebugSegments=20,DebugThickness=0.25)
    DriverAxis=(X=1,Y=0,Z=0)
    DriverItem=(Type=Bone,Name="thigh_r")
    FalloffRegionScaleFactors=(PositiveWidth=1,NegativeWidth=1,PositiveHeight=1,NegativeHeight=1)
    FalloffSize=0.8
    OptionalParentItem=(Type=Bone,Name="spine_04")
    RotationOffset=(X=-90,Y=10,Z=-80)
    ExecutePin=@n5.A
   )
   > n7 = &RigUnit_SphericalPoseReader(
    ActiveRegionScaleFactors=(PositiveWidth=0.7,NegativeWidth=0.4,PositiveHeight=0.3,NegativeHeight=0.35)
    ActiveRegionSize=0.7
    Debug=(bDrawDebug=false,bDraw2D=false,DebugScale=25,DebugSegments=20,DebugThickness=0.25)
    DriverAxis=(X=-1,Y=0,Z=0)
    DriverItem=(Type=Bone,Name="thigh_l")
    FalloffRegionScaleFactors=(PositiveWidth=1,NegativeWidth=1,PositiveHeight=1,NegativeHeight=1)
    FalloffSize=1
    OptionalParentItem=(Type=Bone,Name="spine_03")
    RotationOffset=(X=90,Y=10,Z=-80)
   )
   > n8 = &RigUnit_SphericalPoseReader(
    ActiveRegionScaleFactors=(PositiveWidth=1,NegativeWidth=1,PositiveHeight=1,NegativeHeight=1)
    ActiveRegionSize=0.1
    Debug=(bDrawDebug=false,bDraw2D=false,bDrawLocalAxes=false,DebugScale=25,DebugSegments=20,DebugThickness=0.25)
    DriverAxis=(X=0,Y=-1,Z=0)
    DriverItem=(Type=Bone,Name="deform_skirt_f_m_l")
    FalloffRegionScaleFactors=(PositiveWidth=1,NegativeWidth=1,PositiveHeight=1,NegativeHeight=1)
    FalloffSize=0.3
    OptionalParentItem=(Type=Bone,Name="thigh_l")
    RotationOffset=(X=90,Y=0,Z=-120)
   )
   > &RigUnit_AimConstraintLocalSpaceOffset(
    AimAxis=(X=-1,Y=0,Z=0)
    Child=(Type=Bone,Name="deform_belt_l")
    Filter=(bX=True,bY=True,bZ=True)
    Parents=((Item=(Type=Bone,Name="spine_03"),Weight=1))
    UpAxis=(X=0,Y=-1,Z=0)
    Weight=1
    WorldUp=(Target=(X=0,Y=0,Z=1),Kind=Location,Space=(Type=Bone,Name="deform_belt_l"))
   )
   > &RigUnit_AimConstraintLocalSpaceOffset(
    Child=(Type=Bone,Name="deform_belt_r")
    Filter=(bX=True,bY=True,bZ=True)
    Parents=((Item=(Type=Bone,Name="thigh_r"),Weight=1))
    UpAxis=(X=0,Y=-1,Z=0)
    Weight=1
    WorldUp=(Target=(X=0,Y=0,Z=1),Kind=Location,Space=(Type=Bone,Name="deform_belt_r"))
   )
   > &RigUnit_AimConstraintLocalSpaceOffset(
    Child=(Type=Bone,Name="deform_belt_gun_r")
    Filter=(bX=True,bY=True,bZ=True)
    Parents=((Item=(Type=Bone,Name="thigh_r"),Weight=1))
    UpAxis=(X=0,Y=-1,Z=0)
    Weight=1
    WorldUp=(Target=(X=0,Y=0,Z=1),Kind=Location,Space=(Type=Bone,Name="deform_belt_gun_r"))
    bMaintainOffset=true
   )
   > &RigUnit_ParentConstraint(
    AdvancedSettings=(InterpolationType=Average,RotationOrderForFilter=XZY)
    Child=(Type=Bone,Name="deform_belt_f")
    Filter=(TranslationFilter=(bX=true,bY=false,bZ=true),RotationFilter=(bX=true,bY=true,bZ=true),ScaleFilter=(bX=True,bY=True,bZ=True))
    Parents=((Item=(Type=Bone,Name="deform_belt_l"),Weight=0.5),(Item=(Type=Bone,Name="deform_belt_r"),Weight=0.5))
    Weight=1
    bMaintainOffset=false
   )
   > &RigUnit_ModifyTransforms(
    ItemToModify=((Item=(Type=Bone,Name="dyn_belt_l"),Transform=(Rotation=(X=0,Y=0,Z=0,W=1),Translation=(X=-2,Y=0,Z=0),Scale3D=(X=1,Y=1,Z=1))),(Item=(Type=Bone,Name="deform_belt_f"),Transform=(Rotation=(X=0,Y=0,Z=0,W=1),Translation=(X=-1,Y=0,Z=0),Scale3D=(X=1,Y=1,Z=1))))
    WeightMinimum=0
    Weight=&RigVMFunction_MathDoubleAdd(A=&RigVMFunction_MathDoubleDiv(B=3, A=@n8.OutputParam).Result, B=@n7.OutputParam).Result
   )
   Add = &RigVMFunction_MathDoubleAdd(A=&RigVMFunction_MathDoubleDiv(B=3, A=@n8.OutputParam).Result, B=@n6.OutputParam)
   > &RigUnit_ModifyTransforms(
    ItemToModify="((Item=(Type=Bone,Name=\"dyn_belt_r\"),Transform=(Rotation=(X=0,Y=0,Z=-0.1305,W=0.9914),Translation=(X=-4,Y=-4,Z=2),Scale3D=(X=1,Y=1,Z=1))),(Item=(Type=Bone,Name=\"deform_belt_gun_r\"),Transform=(Rotation=(X=0,Y=0,Z=0,W=1),Translation=(X=-3,Y=-2,Z=0),Scale3D=( ...(truncated)"
    WeightMinimum=0
    Weight=@Add.Result
   )
   > n9 = &RigUnit_SphericalPoseReader(
    ActiveRegionScaleFactors=(PositiveWidth=2,NegativeWidth=1,PositiveHeight=2,NegativeHeight=1)
    ActiveRegionSize=0.1
    Debug=(bDrawDebug=false,bDraw2D=false,DebugScale=25,DebugSegments=20,DebugThickness=0.25)
    DriverAxis=(X=-1,Y=0,Z=0)
    DriverItem=(Type=Bone,Name="thigh_l")
    FalloffRegionScaleFactors=(PositiveWidth=1,NegativeWidth=1,PositiveHeight=1,NegativeHeight=0.2)
    FalloffSize=0.5
    OptionalParentItem=(Type=Bone,Name="spine_03")
    RotationOffset=(X=30,Y=50,Z=-80)
   )
   > &RigUnit_ModifyTransforms(
    ItemToModify=((Item=(Type=Bone,Name="deform_belt_f"),Transform=(Rotation=(X=0,Y=0,Z=0,W=1),Translation=(X=-3,Y=0,Z=0),Scale3D=(X=1,Y=1,Z=1))))
    WeightMinimum=0
    Weight=@n9.OutputParam
   )
   > Get Delta Ref = RigVMFunctionReferenceNode(Item=(Type=Bone,Name="thigh_l"))
   Remap = &RigVMFunction_MathDoubleRemap(SourceMaximum=-110, SourceMinimum=0, TargetMaximum=1, TargetMinimum=0, Value=@"Get Delta Ref".Result.Rotation.Yaw)
   RigVMFunction_AnimEvalRichCurve = &RigVMFunction_AnimEvalRichCurve(
    Curve=(EditorCurveData=(Keys=((),(Time=0.5945,Value=0.5945),(Time=0.783,Value=0.7543),(Time=1,Value=1))))
    SourceMinimum=0
    TargetMinimum=0
    Value=@Remap.Result
   )
   > &RigUnit_ModifyTransforms(
    ItemToModify="((Item=(Type=Bone,Name=\"dyn_chain_1_f\"),Transform=(Rotation=(X=0.194,Y=0.6128,Z=-0.5169,W=0.5654),Translation=(X=0,Y=0,Z=0),Scale3D=(X=1,Y=1,Z=1))),(Item=(Type=Bone,Name=\"dyn_chain_bk\"),Transform=(Rotation=(X=0.067,Y=-0.25,Z=0.25,W=0.933),Translation=(X=0, ...(truncated)"
    WeightMinimum=0
    Weight=@RigVMFunction_AnimEvalRichCurve.Result
   )
   > &RigUnit_ModifyTransforms(
    ItemToModify="((Item=(Type=Bone,Name=\"dyn_chain_1_f\"),Transform=(Rotation=(X=0.0133,Y=0.3263,Z=-0.1208,W=0.9374),Translation=(X=0,Y=0,Z=0),Scale3D=(X=1,Y=1,Z=1))),(Item=(Type=Bone,Name=\"dyn_chain_bk\"),Transform=(Rotation=(X=0.2122,Y=-0.1124,Z=0.0155,W=0.9706),Translatio ...(truncated)"
    WeightMinimum=0
    Weight=&RigVMFunction_MathDoubleRemap(SourceMaximum=60, SourceMinimum=0, TargetMaximum=1, TargetMinimum=0, Value=@"Get Delta Ref".Result.Rotation.Pitch).Result
   )
   > &RigUnit_ParentConstraint(
    AdvancedSettings=(InterpolationType=Average,RotationOrderForFilter=XZY)
    Child=(Type=Bone,Name="dyn_chain_2_f")
    Filter=(TranslationFilter=(bX=true,bY=true,bZ=true),RotationFilter=(bX=true,bY=true,bZ=true),ScaleFilter=(bX=True,bY=True,bZ=True))
    Parents=((Item=(Type=Bone,Name="thigh_l"),Weight=1))
    Weight=0.3
    bMaintainOffset=true
   )
   > &RigUnit_ParentConstraint(
    AdvancedSettings=(InterpolationType=Average,RotationOrderForFilter=XZY)
    Child=(Type=Bone,Name="dyn_chain_3_bk")
    Filter=(TranslationFilter=(bX=true,bY=true,bZ=true),RotationFilter=(bX=true,bY=true,bZ=true),ScaleFilter=(bX=True,bY=True,bZ=True))
    Parents=((Item=(Type=Bone,Name="dyn_chain_4_f"),Weight=1))
    Weight=0.3
    bMaintainOffset=true
   )
   > &RigUnit_AimConstraintLocalSpaceOffset(
    Child=(Type=Bone,Name="dyn_chain_4_f")
    Filter=(bX=True,bY=True,bZ=True)
    Parents=((Item=(Type=Bone,Name="dyn_chain_2_bk"),Weight=1))
    UpAxis=(X=0,Y=-1,Z=0)
    Weight=1
    WorldUp=(Target=(X=0,Y=0,Z=1),Kind=Location,Space=(Type=Bone,Name="dyn_chain_4_f"))
    bMaintainOffset=true
   )
   > Set Transform_12 = /Script/ControlRig.RigUnit_SetTranslation(
    Item=(Type=Bone,Name="dyn_chain_2_f")
    Space=GlobalSpace
    Weight=1
    bInitial=false
    bPropagateToChildren=true
    Value=/Script/RigVM.RigVMFunction_MathTransformClampSpatially(
     Axis=X
     DebugColor=(R=1,G=0,B=0,A=1)
     DebugThickness=1
     Maximum=11
     Minimum=8
     Type=Cylinder
     bDrawDebug=false
     Value=&RigUnit_GetTransform(Item=(Type=Bone,Name="dyn_chain_2_f"), bInitial=false).Transform
     Space=&RigUnit_GetTransform(Item=(Type=Control,Name="thigh_l_ctrl"), bInitial=false).Transform
    ) {
     is=struct RigVMUnitNode
     template="ClampSpatially::Execute(in Value,in Axis,in Type,in Minimum,in Maximum,in Space,in bDrawDebug,in DebugColor,in DebugThickness,out Result)"
    }.Result.Translation
   ) {is=struct RigVMUnitNode, template="Set Transform::Execute(in Item,in Space,in bInitial,in Value,in Weight,in bPropagateToChildren)"}
   > Set Transform_2 = &RigUnit_SetTranslation(
    Item=(Type=Bone,Name="dyn_chain_3_f")
    Space=GlobalSpace
    Weight=1
    bInitial=false
    bPropagateToChildren=true
    Value=&RigVMFunction_MathTransformClampSpatially(
     Axis=X
     DebugColor=(R=1,G=0,B=0,A=1)
     DebugThickness=1
     Maximum=11
     Minimum=9
     Type=Cylinder
     bDrawDebug=false
     Value=&RigUnit_GetTransform(Item=(Type=Bone,Name="dyn_chain_3_f"), bInitial=false).Transform
     Space=&RigUnit_GetTransform(Item=(Type=Control,Name="thigh_l_ctrl"), bInitial=false).Transform
    ) {template="ClampSpatially::Execute(in Value,in Axis,in Type,in Minimum,in Maximum,in Space,in bDrawDebug,in DebugColor,in DebugThickness,out Result)"}.Result.Translation
   ) {template="Set Transform::Execute(in Item,in Space,in bInitial,in Value,in Weight,in bPropagateToChildren)"}
   > Set Transform_3 = &RigUnit_SetTranslation(
    Item=(Type=Bone,Name="dyn_chain_1_bk")
    Space=GlobalSpace
    Weight=1
    bInitial=false
    bPropagateToChildren=true
    Value=&RigVMFunction_MathTransformClampSpatially(
     Axis=X
     DebugColor=(R=1,G=0,B=0,A=1)
     DebugThickness=1
     Maximum=11
     Minimum=9
     Type=Cylinder
     bDrawDebug=false
     Value=&RigUnit_GetTransform(Item=(Type=Bone,Name="dyn_chain_1_bk"), bInitial=false).Transform
     Space=&RigUnit_GetTransform(Item=(Type=Control,Name="thigh_l_ctrl"), bInitial=false).Transform
    ) {template="ClampSpatially::Execute(in Value,in Axis,in Type,in Minimum,in Maximum,in Space,in bDrawDebug,in DebugColor,in DebugThickness,out Result)"}.Result.Translation
   ) {template="Set Transform::Execute(in Item,in Space,in bInitial,in Value,in Weight,in bPropagateToChildren)"}
   n10 = &RigUnit_SphericalPoseReader(
    ActiveRegionScaleFactors=(PositiveWidth=0.7,NegativeWidth=0.4,PositiveHeight=0.3,NegativeHeight=0.35)
    ActiveRegionSize=0.3
    Debug=(bDrawDebug=false,bDraw2D=false,DebugScale=25,DebugSegments=20,DebugThickness=0.25)
    DriverAxis=(X=-1,Y=0,Z=0)
    DriverItem=(Type=Bone,Name="thigh_l")
    FalloffRegionScaleFactors=(PositiveWidth=1,NegativeWidth=1,PositiveHeight=1,NegativeHeight=1)
    FalloffSize=0.8
    OptionalParentItem=(Type=Bone,Name="spine_03")
    RotationOffset=(X=90,Y=10,Z=-80)
    ExecutePin=@n5.B
   )
   > n11 = &RigUnit_SphericalPoseReader(
    ActiveRegionScaleFactors=(PositiveWidth=1,NegativeWidth=1,PositiveHeight=1,NegativeHeight=1)
    ActiveRegionSize=0.2
    Debug=(bDrawDebug=false,bDraw2D=false,DebugScale=25,DebugSegments=20,DebugThickness=0.25)
    DriverAxis=(X=0,Y=1,Z=0)
    DriverItem=(Type=Bone,Name="spine_05")
    FalloffRegionScaleFactors=(PositiveWidth=1,NegativeWidth=1,PositiveHeight=1,NegativeHeight=1)
    FalloffSize=0.5
    OptionalParentItem=(Type=Bone,Name="thigh_l")
    RotationOffset=(X=180,Y=10,Z=-80)
   )
   > n12 = &RigUnit_SphericalPoseReader(
    ActiveRegionScaleFactors=(PositiveWidth=1,NegativeWidth=1,PositiveHeight=1,NegativeHeight=1)
    ActiveRegionSize=0.1
    Debug=(bDrawDebug=false,bDraw2D=false,bDrawLocalAxes=false,DebugScale=25,DebugSegments=20,DebugThickness=0.25)
    DriverAxis=(X=0,Y=-1,Z=0)
    DriverItem=(Type=Bone,Name="deform_skirt_f_m_l")
    FalloffRegionScaleFactors=(PositiveWidth=2,NegativeWidth=1,PositiveHeight=2,NegativeHeight=2)
    FalloffSize=0.3
    OptionalParentItem=(Type=Bone,Name="thigh_l")
    RotationOffset=(X=90,Y=0,Z=-120)
   )
   > n13 = &RigUnit_SphericalPoseReader(
    ActiveRegionScaleFactors=(PositiveWidth=1,NegativeWidth=1,PositiveHeight=1,NegativeHeight=1)
    ActiveRegionSize=0.2
    Debug=(bDrawDebug=false,bDraw2D=false,DebugScale=25,DebugSegments=20,DebugThickness=0.25)
    DriverAxis=(X=0,Y=1,Z=0)
    DriverItem=(Type=Bone,Name="spine_04")
    FalloffRegionScaleFactors=(PositiveWidth=1,NegativeWidth=1,PositiveHeight=1,NegativeHeight=1)
    FalloffSize=0.5
    OptionalParentItem=(Type=Bone,Name="thigh_l")
    RotationOffset=(X=180,Y=10,Z=-80)
   )
   Divide = &RigVMFunction_MathDoubleDiv(B=1, A=@n12.OutputParam)
   Subtract = &RigVMFunction_MathDoubleSub(B=@Divide.Result, A=@n10.OutputParam)
   Clamp = &RigVMFunction_MathDoubleClamp(Value=@Subtract.Result)
   > &RigUnit_ModifyTransforms(
    ItemToModify="((Item=(Type=Bone,Name=\"deform_skirt_f_l\"),Transform=(Rotation=(X=-0.3303,Y=-0.449,Z=0.2783,W=-0.7822),Translation=(X=-0.0001,Y=0,Z=-10),Scale3D=(X=1,Y=1,Z=1))),(Item=(Type=Bone,Name=\"deform_skirt_f_m_l\"),Transform=(Rotation=(X=0.3962,Y=0.1147,Z=-0.2534,W= ...(truncated)"
    WeightMinimum=0
    Weight=&RigVMFunction_MathDoubleSub(B=@n11.OutputParam, A=&RigVMFunction_MathDoubleSub(A=@Clamp.Result, B=@n13.OutputParam).Result).Result
   )
   RigVMFunction_MathVectorDistance = &RigVMFunction_MathVectorDistance(
    A=&RigUnit_GetTransform(Item=(Type=Bone,Name="deform_skirt_f_m_l"), bInitial=false).Transform.Translation
    B=&RigUnit_GetTransform(Item=(Type=Bone,Name="thigh_twist_01_l"), bInitial=false).Transform.Translation
   )
   > &RigUnit_ModifyTransforms(
    ItemToModify="((Item=(Type=Bone,Name=\"deform_skirt_f_l\"),Transform=(Rotation=(X=0.9537,Y=0.1405,Z=-0.1201,W=0.2372),Translation=(X=4.0615,Y=4.6205,Z=-3.7933),Scale3D=(X=1,Y=1,Z=1))),(Item=(Type=Bone,Name=\"deform_skirt_f_m_l\"),Transform=(Rotation=(X=0.8654,Y=-0.0302,Z=-0 ...(truncated)"
    WeightMinimum=0
    Weight=&RigVMFunction_MathDoubleClamp(Value=&RigVMFunction_MathDoubleRemap(SourceMaximum=24, SourceMinimum=14, TargetMaximum=0, TargetMinimum=1, bClamp=False, Value=@RigVMFunction_MathVectorDistance.Result).Result).Result
   )
   &RigUnit_AimConstraintLocalSpaceOffset(
    Child=(Type=Bone,Name="dyn_gun")
    Filter=(bX=True,bY=True,bZ=True)
    Parents=((Item=(Type=Bone,Name="calf_r"),Weight=1))
    UpAxis=(X=0,Y=-1,Z=0)
    Weight=1
    WorldUp=(Target=(X=0,Y=0,Z=1),Kind=Location,Space=(Type=Bone,Name="dyn_gun"))
    ExecutePin=@RigVMFunction_Sequence.D
   )
   > &RigUnit_ParentConstraint(
    AdvancedSettings=(RotationOrderForFilter=XZY)
    Child=(Type=Bone,Name="deform_chain_f")
    Parents=((Item=(Type=Bone,Name="dyn_belt_r"),Weight=0.1694),(Item=(Type=Bone,Name="dyn_belt_l"),Weight=0.2401),(Item=(Type=Bone,Name="dyn_belt_f"),Weight=0.6))
    Weight=1
   )
   &RigUnit_ModifyTransforms(
    ExecutePin=()
    ItemToModify="((Item=(Type=Bone,Name=\"dyn_chain_1_f\"),Transform=(Rotation=(X=0.1312,Y=0.6293,Z=-0.5713,W=0.5103),Translation=(X=0,Y=0,Z=0),Scale3D=(X=1,Y=1,Z=1))),(Item=(Type=Bone,Name=\"dyn_chain_bk\"),Transform=(Rotation=(X=0.067,Y=-0.25,Z=0.25,W=0.933),Translation=(X=0 ...(truncated)"
    Weight=1
    WeightMinimum=0
   )
   &RigUnit_ModifyTransforms(
    ExecutePin=()
    ItemToModify="((Item=(Type=Bone,Name=\"deform_skirt_b_l\"),Transform=(Rotation=(X=0,Y=0,Z=-0.9659,W=0.2588),Translation=(X=0,Y=0,Z=0),Scale3D=(X=1,Y=1,Z=1))),(Item=(Type=Bone,Name=\"deform_skirt_b_r\"),Transform=(Rotation=(X=0,Y=0,Z=-0.9659,W=0.2588),Translation=(X=0,Y=0,Z= ...(truncated)"
    Weight=1
    WeightMinimum=0
   )
   AlphaInterp = &RigVMFunction_AlphaInterp(
    InterpSpeedDecreasing=10
    InterpSpeedIncreasing=10
    Scale=1
    bClampResult=False
    Value=&RigVMFunction_MathDoubleRemap(SourceMaximum=-20, SourceMinimum=0, TargetMaximum=1, TargetMinimum=0, Value=@"Get Delta Ref".Result.Rotation.Pitch).Result
   )
   &RigUnit_ModifyTransforms(
    ExecutePin=()
    ItemToModify=((Item=(Type=Bone,Name="dyn_chain_1_f"),Transform=(Rotation=(X=0,Y=0,Z=0,W=1),Translation=(X=0,Y=0,Z=0),Scale3D=(X=1,Y=1,Z=1))),(Item=(Type=Bone,Name="dyn_chain_bk"),Transform=(Rotation=(X=0,Y=0,Z=0,W=1),Translation=(X=0,Y=0,Z=0),Scale3D=(X=1,Y=1,Z=1))))
    WeightMinimum=0
    Weight=@AlphaInterp.Result
   )
   RigVMDispatchNode(Enabled=True, ExecuteContext=(), ScreenColor=(R=1,G=1,B=1,A=1), ScreenDuration=0.0, Value=@Remap.Result) {template="DISPATCH_RigVMDispatch_Print(in Prefix,in Value,in Enabled,in ScreenDuration,in ScreenColor)"}
  ]}
]
hierarchy [
 root {loc=<0 0 0>, rot=<0 0 0 1>}
 pelvis {parent=root, loc=<0 -0.5301 98.0571>, rot=<0.0127 0.707 -0.0127 -0.707>}
 spine_01 {parent=pelvis, loc=<6.1208 -0.9328 0>, rot=<0 0 0.034 -0.9994>}
 spine_02 {parent=spine_01, loc=<6.5997 -0.2095 0>, rot=<0 0 0.0238 -0.9997>}
 spine_03 {parent=spine_02, loc=<6.4914 -0.1237 0>, rot=<0 0 0.0149 -0.9999>}
 spine_04 {parent=spine_03, loc=<6.9598 0.0794 0>, rot=<0 0 -0.0132 -0.9999>}
 spine_05 {parent=spine_04, loc=<6.9987 0 0>, rot=<0 0 -0.0015 -1>}
 neck_01 {parent=spine_05, loc=<11.3138 0.1529 0>, rot=<0 0 -0.2003 -0.9797>}
 neck_02 {parent=neck_01, loc=<4.5627 0 0>, rot=<0 0 0.0504 -0.9987>}
 head {parent=neck_02, loc=<5.3914 0 0>, rot=<0 0 0.115 -0.9934>}
 clavicle_l {parent=spine_05, loc=<5.4831 -0.1053 -3.2946>, rot=<0.0465 -0.7289 -0.009 -0.6829>}
 upperarm_l {parent=clavicle_l, loc=<9.8658 -1.2136 -0.4105>, rot=<0.0164 -0.3245 -0.0269 -0.9454>}
 lowerarm_l {parent=upperarm_l, loc=<26.465 -0.8345 0.259>, rot=<0.0036 0.0161 -0.2298 -0.9731>}
 hand_l {parent=lowerarm_l, loc=<22.7355 -0.3203 -0.6105>, rot=<-0.6784 -0.0223 0.0735 -0.7307>}
 thumb_01_l {parent=hand_l, loc=<1.3118 -0.3191 -2.459>, rot=<0.6546 -0.4042 0.0664 -0.6354>}
 thumb_02_l {parent=thumb_01_l, loc=<5.9336 -0.0448 0.0401>, rot=<-0.0037 0.0213 0.298 -0.9543>}
 thumb_03_l {parent=thumb_02_l, loc=<2.2223 0.3682 -0.0907>, rot=<0.0118 -0.013 0.0066 -0.9998>}
 index_metacarpal_l {parent=hand_l, loc=<1.6535 0.0183 -2.96>, rot=<0.0158 -0.0124 -0.0173 -0.9996>}
 index_01_l {parent=index_metacarpal_l, loc=<8.3008 0.0357 -0.0524>, rot=<-0.0044 0.0485 0.0924 -0.9945>}
 index_02_l {parent=index_01_l, loc=<3.3587 -0.3004 0.0452>, rot=<-0.002 0.0027 0.3225 -0.9466>}
 index_03_l {parent=index_02_l, loc=<2.5578 -0.123 -0.0252>, rot=<-0.002 0.0067 0.2558 -0.9667>}
 middle_metacarpal_l {parent=hand_l, loc=<2.1856 -0.1725 -0.5984>, rot=<0.006 0.0168 -0.0329 -0.9993>}
 middle_01_l {parent=middle_metacarpal_l, loc=<7.3275 -0.2447 0.0129>, rot=<-0.0208 0.0042 0.0858 -0.9961>}
 middle_02_l {parent=middle_01_l, loc=<4.0123 0.1401 0.1175>, rot=<-0.0218 0.0258 0.3512 -0.9357>}
 middle_03_l {parent=middle_02_l, loc=<2.6077 0.0534 0.1373>, rot=<0.0082 0.0188 0.1708 -0.9851>}
 ring_metacarpal_l {parent=hand_l, loc=<2.2074 -0.3625 1.3638>, rot=<-0.0284 0.0722 -0.0458 -0.9959>}
 ring_01_l {parent=ring_metacarpal_l, loc=<7.0295 -0.6197 -0.1446>, rot=<-0.0489 -0.0252 0.1892 -0.9804>}
 ring_02_l {parent=ring_01_l, loc=<3.6174 0.4647 -0.1027>, rot=<0 0 0.2806 -0.9598>}
 ring_03_l {parent=ring_02_l, loc=<2.4152 0.114 0.0199>, rot=<0 0 0.1189 -0.9929>}
 pinky_metacarpal_l {parent=hand_l, loc=<1.7616 -0.9753 3.3814>, rot=<-0.0397 0.0814 -0.069 -0.9935>}
 pinky_01_l {parent=pinky_metacarpal_l, loc=<7.1709 -0.4831 -0.1094>, rot=<-0.0654 -0.0035 0.2242 -0.9723>}
 pinky_02_l {parent=pinky_01_l, loc=<2.8274 0.0893 -0.1043>, rot=<0.0037 0.0011 0.2854 -0.9584>}
 pinky_03_l {parent=pinky_02_l, loc=<2.2046 0.0307 -0.0078>, rot=<-0.0002 0.0098 0.1445 -0.9895>}
 deform_wrist_dn_base_l {parent=hand_l, loc=<0 -1.5326 0>, rot=<0 0 0 -1>}
 deform_wrist_dn_l {parent=deform_wrist_dn_base_l, loc=<0 0 0>, rot=<0 0 0 -1>}
 deform_wrist_up_base_l {parent=hand_l, loc=<0 2.3009 0>, rot=<0 0 0 -1>}
 deform_wrist_up_l {parent=deform_wrist_up_base_l, loc=<0 0 0>, rot=<0 0 0 -1>}
 lowerarm_twist_02_l {parent=lowerarm_l, loc=<10.6707 0 0>, rot=<0 0 0 1>}
 lowerarm_twist_01_l {parent=lowerarm_l, loc=<18.316 0 0>, rot=<0 0 0 1>}
 deform_elbow_in_base_l {parent=lowerarm_l, loc=<0.1834 3.6062 -0.2572>, rot=<0 0 0 -1>}
 deform_elbow_in_l {parent=deform_elbow_in_base_l, loc=<0 0 0>, rot=<0 0 0 -1>}
 deform_elbow_out_base_l {parent=lowerarm_l, loc=<0 -4 0>, rot=<0 0 0 1>}
 deform_elbow_out_l {parent=deform_elbow_out_base_l, loc=<0 0 0>, rot=<0 0 0 1>}
 upperarm_twist_01_l {parent=upperarm_l, loc=<7.2135 -0.1134 0>, rot=<0 0 0 1>}
 deform_upperarm_bk_base_l {parent=upperarm_twist_01_l, loc=<-6.4053 -4.3746 0.314>, rot=<0 0 0 -1>}
 deform_upperarm_bk_l {parent=deform_upperarm_bk_base_l, loc=<0 0 0>, rot=<0 0 0 -1>}
 deform_upperarm_fr_base_l {parent=upperarm_twist_01_l, loc=<-5.6312 5.092 0.3354>, rot=<0 0 0 -1>}
 deform_upperarm_fr_l {parent=deform_upperarm_fr_base_l, loc=<0 0 0>, rot=<0 0 0 -1>}
 deform_upperarm_up_base_l {parent=upperarm_twist_01_l, loc=<-8.891 0.4825 3.9517>, rot=<0 0 0 -1>}
 deform_upperarm_up_l {parent=deform_upperarm_up_base_l, loc=<0 0 0>, rot=<0 0 0 -1>}
 upperarm_twist_02_l {parent=upperarm_l, loc=<19.7998 -0.1134 0>, rot=<0 0 0 1>}
 deform_collar_l {parent=clavicle_l, loc=<4.4496 7.9068 5.6378>, rot=<-0.544 -0.0501 0.6737 0.4977>}
 dyn_collar_l {parent=deform_collar_l, loc=<0 0 0>, rot=<0 0 0 1>}
 clavicle_r {parent=spine_05, loc=<5.4831 -0.1053 3.2946>, rot=<0.7289 0.0465 -0.6829 0.009>}
 upperarm_r {parent=clavicle_r, loc=<-9.8658 1.2136 0.4105>, rot=<0.0164 -0.3245 -0.0269 -0.9454>}
 lowerarm_r {parent=upperarm_r, loc=<-26.465 0.8345 -0.259>, rot=<0.0036 0.0161 -0.2298 -0.9731>}
 hand_r {parent=lowerarm_r, loc=<-22.7355 0.3203 0.6106>, rot=<-0.6784 -0.0223 0.0735 -0.7307>}
 thumb_01_r {parent=hand_r, loc=<-1.3118 0.3191 2.459>, rot=<0.6546 -0.4042 0.0664 -0.6354>}
 thumb_02_r {parent=thumb_01_r, loc=<-5.9336 0.0448 -0.0401>, rot=<-0.0037 0.0213 0.298 -0.9543>}
 thumb_03_r {parent=thumb_02_r, loc=<-2.2223 -0.3682 0.0907>, rot=<0.0118 -0.013 0.0066 -0.9998>}
 index_metacarpal_r {parent=hand_r, loc=<-1.6535 -0.0183 2.96>, rot=<0.0158 -0.0124 -0.0173 -0.9996>}
 index_01_r {parent=index_metacarpal_r, loc=<-8.3008 -0.0357 0.0524>, rot=<-0.0044 0.0485 0.0924 -0.9945>}
 index_02_r {parent=index_01_r, loc=<-3.3587 0.3004 -0.0452>, rot=<-0.002 0.0027 0.3225 -0.9466>}
 index_03_r {parent=index_02_r, loc=<-2.5578 0.123 0.0252>, rot=<-0.002 0.0067 0.2558 -0.9667>}
 middle_metacarpal_r {parent=hand_r, loc=<-2.1856 0.1725 0.5984>, rot=<0.006 0.0168 -0.0329 -0.9993>}
 middle_01_r {parent=middle_metacarpal_r, loc=<-7.3275 0.2447 -0.0129>, rot=<-0.0208 0.0042 0.0858 -0.9961>}
 middle_02_r {parent=middle_01_r, loc=<-4.0123 -0.1401 -0.1175>, rot=<-0.0218 0.0258 0.3512 -0.9357>}
 middle_03_r {parent=middle_02_r, loc=<-2.6077 -0.0534 -0.1373>, rot=<0.0082 0.0188 0.1708 -0.9851>}
 ring_metacarpal_r {parent=hand_r, loc=<-2.2074 0.3625 -1.3638>, rot=<-0.0284 0.0722 -0.0458 -0.9959>}
 ring_01_r {parent=ring_metacarpal_r, loc=<-7.0295 0.6197 0.1446>, rot=<-0.0489 -0.0252 0.1892 -0.9804>}
 ring_02_r {parent=ring_01_r, loc=<-3.6174 -0.4647 0.1027>, rot=<0 0 0.2806 -0.9598>}
 ring_03_r {parent=ring_02_r, loc=<-2.4152 -0.114 -0.0199>, rot=<0 0 0.1189 -0.9929>}
 pinky_metacarpal_r {parent=hand_r, loc=<-1.7616 0.9753 -3.3814>, rot=<-0.0397 0.0814 -0.069 -0.9935>}
 pinky_01_r {parent=pinky_metacarpal_r, loc=<-7.1709 0.4831 0.1094>, rot=<-0.0654 -0.0035 0.2242 -0.9723>}
 pinky_02_r {parent=pinky_01_r, loc=<-2.8274 -0.0893 0.1043>, rot=<0.0037 0.0011 0.2854 -0.9584>}
 pinky_03_r {parent=pinky_02_r, loc=<-2.2046 -0.0307 0.0078>, rot=<-0.0002 0.0098 0.1445 -0.9895>}
 deform_wrist_dn_base_r {parent=hand_r, loc=<-0.0001 1.5328 0>, rot=<0 0 0 -1>}
 deform_wrist_dn_r {parent=deform_wrist_dn_base_r, loc=<0 0 0>, rot=<0 0 0 -1>}
 deform_wrist_up_base_r {parent=hand_r, loc=<-0.0002 -2.3007 -0.0001>, rot=<0 0 0 -1>}
 deform_wrist_up_r {parent=deform_wrist_up_base_r, loc=<0 0 0>, rot=<0 0 0 -1>}
 lowerarm_twist_02_r {parent=lowerarm_r, loc=<-10.6707 0 0>, rot=<0 0 0 -1>}
 lowerarm_twist_01_r {parent=lowerarm_r, loc=<-18.316 0 0>, rot=<0 0 0 -1>}
 deform_elbow_in_base_r {parent=lowerarm_r, loc=<-0.1834 -3.6062 0.2572>, rot=<0 0 0 -1>}
 deform_elbow_in_r {parent=deform_elbow_in_base_r, loc=<0 0 0>, rot=<0 0 0 -1>}
 deform_elbow_out_base_r {parent=lowerarm_r, loc=<0 4 -0.0001>, rot=<0 0 0 -1>}
 deform_elbow_out_r {parent=deform_elbow_out_base_r, loc=<0 0 0>, rot=<0 0 0 1>}
 upperarm_twist_01_r {parent=upperarm_r, loc=<-7.2135 0.1134 0>, rot=<0 0 0 -1>}
 deform_upperarm_bk_base_r {parent=upperarm_twist_01_r, loc=<6.4053 4.3746 -0.314>, rot=<0 0 0 -1>}
 deform_upperarm_bk_r {parent=deform_upperarm_bk_base_r, loc=<0 0 0>, rot=<0 0 0 -1>}
 deform_upperarm_fr_base_r {parent=upperarm_twist_01_r, loc=<5.6312 -5.092 -0.3354>, rot=<0 0 0 -1>}
 deform_upperarm_fr_r {parent=deform_upperarm_fr_base_r, loc=<0 0 0>, rot=<0 0 0 -1>}
 deform_upperarm_up_base_r {parent=upperarm_twist_01_r, loc=<8.891 -0.4825 -3.9517>, rot=<0 0 0 -1>}
 deform_upperarm_up_r {parent=deform_upperarm_up_base_r, loc=<0 0 0>, rot=<0 0 0 -1>}
 upperarm_twist_02_r {parent=upperarm_r, loc=<-19.7998 0.1134 0>, rot=<0 0 0 -1>}
 deform_collar_r {parent=clavicle_r, loc=<-4.4496 -7.9068 -5.6382>, rot=<-0.0501 0.544 0.4977 -0.6737>}
 dyn_collar_r {parent=deform_collar_r, loc=<0 0 0>, rot=<0 0 0 1>}
 attach_backpack {parent=spine_05, loc=<-0.123 -13.473 0>, rot=<-0.0093 -0.707 0.0093 -0.707>}
 attach_cape {parent=spine_05, loc=<9.017 -11.714 0>, rot=<0.0616 -0.7044 -0.0616 -0.7044>}
 deform_pec_aim_base_l {parent=spine_04, loc=<3.2456 9.3309 -1.2237>, rot=<0.0603 -0.4033 0.3318 -0.8507>}
 deform_pec_aim_l {parent=deform_pec_aim_base_l, loc=<0 0 0>, rot=<0 0 0 -1>}
 deform_pec_base_l {parent=deform_pec_aim_l, loc=<9.3665 0 0>, rot=<0 0 0 -1>}
 deform_pec_l {parent=deform_pec_base_l, loc=<0 0 0>, rot=<0 0 0 -1>}
 deform_pec_aim_base_r {parent=spine_04, loc=<3.2452 9.3309 1.2237>, rot=<-0.4033 -0.0603 0.8507 0.3318>}
 deform_pec_aim_r {parent=deform_pec_aim_base_r, loc=<0 0 0>, rot=<0 0 0 -1>}
 deform_pec_base_r {parent=deform_pec_aim_r, loc=<-9.3667 -0.0003 -0.0004>, rot=<0 0 0 -1>}
 deform_pec_r {parent=deform_pec_base_r, loc=<0 0 0>, rot=<0 0 0 -1>}
 deform_skirt_f_l {parent=spine_04, loc=<1.0837 12.048 -8.2171>, rot=<-0.0234 0.3442 0.9335 -0.0979>}
 dyn_skirt_f_l {parent=deform_skirt_f_l, loc=<0 0 0>, rot=<0 0 0 1>}
 dyn_skirt_f_1_l {parent=dyn_skirt_f_l, loc=<11.7406 0 0>, rot=<-0.0022 -0.012 0.1197 -0.9927>}
 dyn_skirt_f_2_l {parent=dyn_skirt_f_1_l, loc=<8.8906 -0.2799 0>, rot=<0 0 0.0226 -0.9997>}
 dyn_skirt_f_3_l {parent=dyn_skirt_f_2_l, loc=<8.7755 0 0>, rot=<0 0 0 1>}
 deform_skirt_f_m_l {parent=spine_04, loc=<1.4875 7.6111 -11.8301>, rot=<-0.0012 0.5058 0.8534 -0.126>}
 dyn_skirt_f_m_l {parent=deform_skirt_f_m_l, loc=<0 0 0>, rot=<0 0 0.0069 -1>}
 dyn_skirt_f_m_1_l {parent=dyn_skirt_f_m_l, loc=<11.6592 0 0>, rot=<0 0 0.1856 -0.9826>}
 dyn_skirt_f_m_2_l {parent=dyn_skirt_f_m_1_l, loc=<8.8071 0.0394 0>, rot=<0 0 0.002 -1>}
 dyn_skirt_f_m_3_l {parent=dyn_skirt_f_m_2_l, loc=<10.4239 0 0>, rot=<0 0 0 1>}
 deform_skirt_f_m_r {parent=spine_04, loc=<1.4871 7.6111 11.8301>, rot=<0.0012 -0.5058 0.8534 -0.126>}
 dyn_skirt_f_m_r {parent=deform_skirt_f_m_r, loc=<0 0 0>, rot=<0 0 0.0069 -1>}
 dyn_skirt_f_m_1_r {parent=dyn_skirt_f_m_r, loc=<11.6584 0.0001 0>, rot=<0 0 0.1856 -0.9826>}
 dyn_skirt_f_m_2_r {parent=dyn_skirt_f_m_1_r, loc=<8.8072 0.0394 0>, rot=<0 0 0.002 -1>}
 dyn_skirt_f_m_3_r {parent=dyn_skirt_f_m_2_r, loc=<10.4241 0.0001 0>, rot=<0 0 0 1>}
 deform_skirt_f_r {parent=spine_04, loc=<1.0839 12.0481 8.2171>, rot=<0.0234 -0.3442 0.9335 -0.0979>}
 dyn_skirt_f_r {parent=deform_skirt_f_r, loc=<0 0 0>, rot=<0 0 0 1>}
 dyn_skirt_f_1_r {parent=dyn_skirt_f_r, loc=<11.7405 0 0>, rot=<0.0022 0.012 0.1197 -0.9927>}
 dyn_skirt_f_2_r {parent=dyn_skirt_f_1_r, loc=<8.8908 -0.2799 0>, rot=<0 0 0.0226 -0.9997>}
 dyn_skirt_f_3_r {parent=dyn_skirt_f_2_r, loc=<8.7755 0 0>, rot=<0 0 0 1>}
 deform_skirt_m_l {parent=spine_03, loc=<-2.2979 -0.5851 -11.0741>, rot=<0.135 -0.7193 -0.6792 -0.0558>}
 dyn_skirt_m_l {parent=deform_skirt_m_l, loc=<0 0 0>, rot=<0 0 0 1>}
 dyn_skirt_m_1_l {parent=dyn_skirt_m_l, loc=<10.3527 -0.0039 0>, rot=<0 0 -0.0008 -1>}
 dyn_skirt_m_2_l {parent=dyn_skirt_m_1_l, loc=<11.1562 0.0299 0>, rot=<0 0 0.002 -1>}
 deform_skirt_m_r {parent=spine_03, loc=<-2.2976 -0.5851 11.0741>, rot=<0.135 -0.7193 0.6792 0.0558>}
 dyn_skirt_m_r {parent=deform_skirt_m_r, loc=<0 0 0>, rot=<0 0 0 1>}
 dyn_skirt_m_1_r {parent=dyn_skirt_m_r, loc=<10.3532 -0.0038 0>, rot=<0 0 -0.0008 -1>}
 dyn_skirt_m_2_r {parent=dyn_skirt_m_1_r, loc=<11.156 0.0299 0>, rot=<0 0 0.002 -1>}
 deform_skirt_b_l {parent=spine_02, loc=<-0.2069 -6.627 -10.344>, rot=<0.132 -0.8843 -0.447 -0.0297>}
 dyn_skirt_b_l {parent=deform_skirt_b_l, loc=<0 0 0>, rot=<0 0 0 1>}
 dyn_skirt_b_1_l {parent=dyn_skirt_b_l, loc=<10.3527 -0.0039 0>, rot=<0 0 -0.0008 -1>}
 dyn_skirt_b_2_l {parent=dyn_skirt_b_1_l, loc=<11.1562 0.0299 0>, rot=<0 0 0.002 -1>}
 deform_skirt_bk_l {parent=spine_02, loc=<-2.5236 -10.8138 -6.0453>, rot=<0.1326 -0.9808 -0.1431 -0.0087>}
 dyn_skirt_bk_l {parent=deform_skirt_bk_l, loc=<0 0 0>, rot=<0 0 0 1>}
 dyn_skirt_bk_1_l {parent=dyn_skirt_bk_l, loc=<11.6537 -0.0261 0>, rot=<0 0 -0.0605 -0.9982>}
 dyn_skirt_bk_2_l {parent=dyn_skirt_bk_1_l, loc=<14.0044 -0.0046 0>, rot=<0 0 0.002 -1>}
 deform_skirt_bk_r {parent=spine_02, loc=<-2.5232 -10.8137 6.0453>, rot=<0.1326 -0.9808 0.1431 0.0087>}
 dyn_skirt_bk_r {parent=deform_skirt_bk_r, loc=<0 0 0>, rot=<0 0 0 -1>}
 dyn_skirt_bk_1_r {parent=dyn_skirt_bk_r, loc=<11.6541 -0.026 0>, rot=<0 0 -0.0605 -0.9982>}
 dyn_skirt_bk_2_r {parent=dyn_skirt_bk_1_r, loc=<14.0044 -0.0046 0>, rot=<0 0 0.002 -1>}
 deform_skirt_b_r {parent=spine_02, loc=<-0.2074 -6.6271 10.344>, rot=<0.132 -0.8843 0.447 0.0297>}
 dyn_skirt_b_r {parent=deform_skirt_b_r, loc=<0 0 0>, rot=<0 0 0 1>}
 dyn_skirt_b_1_r {parent=dyn_skirt_b_r, loc=<10.3522 -0.004 0>, rot=<0 0 -0.0008 -1>}
 dyn_skirt_b_2_r {parent=dyn_skirt_b_1_r, loc=<11.1562 0.0299 0>, rot=<0 0 0.002 -1>}
 thigh_l {parent=pelvis, loc=<-1.0855 -0.2109 -8.7053>, rot=<0.0009 0.0355 0.0155 -0.9993>}
 calf_l {parent=thigh_l, loc=<-42.5284 -0.0019 -0.7297>, rot=<0.0006 -0.0161 -0.0391 -0.9991>}
 foot_l {parent=calf_l, loc=<-46.7401 -2.5014 -0.592>, rot=<0.0427 -0.0238 0.0353 -0.9982>}
 ball_l {parent=foot_l, loc=<-3.5658 10.8066 -0.1544>, rot=<0.0001 -0.0001 -0.7063 -0.7079>}
 calf_twist_02_l {parent=calf_l, loc=<-19.1546 -1.7943 -0.2638>, rot=<0.0237 0.0095 -0.0035 -0.9997>}
 calf_twist_01_l {parent=calf_l, loc=<-35.8483 -1.9197 -0.5794>, rot=<0.0237 0.0095 -0.0035 -0.9997>}
 deform_knee_in_base_l {parent=calf_l, loc=<0.8031 -5.0231 0.2456>, rot=<0 0 0 -1>}
 deform_knee_in_l {parent=deform_knee_in_base_l, loc=<0 0 0>, rot=<0 0 0 -1>}
 deform_knee_out_base_l {parent=calf_l, loc=<0 4 0>, rot=<0 0 0 -1>}
 deform_knee_out_l {parent=deform_knee_out_base_l, loc=<0 0 0>, rot=<0 0 0 1>}
 thigh_twist_01_l {parent=thigh_l, loc=<-27.952 0 0>, rot=<0.0013 0 0 -1>}
 deform_groin_base_l {parent=thigh_l, loc=<0.4801 5.6776 -0.2486>, rot=<0 0 0 -1>}
 deform_groin_l {parent=deform_groin_base_l, loc=<0 0 0>, rot=<0 0 0 -1>}
 deform_zipper_l {parent=thigh_l, loc=<-30.5805 5.1273 -2.311>, rot=<0.0064 0.2356 0.9709 -0.0417>}
 dyn_zipper_l {parent=deform_zipper_l, loc=<0 0 0>, rot=<0 0 0 1>}
 thigh_l_ctrl {controlType=EULER_TRANSFORM, parent=thigh_l, type=Control, loc=<0 -3 0>, rot=<0 0 0 1>}
 thigh_r {parent=pelvis, loc=<-1.0855 -0.2109 8.7053>, rot=<0.0355 -0.0009 0.9993 0.0155>}
 calf_r {parent=thigh_r, loc=<42.5284 0.0019 0.7297>, rot=<0.0006 -0.0161 -0.0391 -0.9991>}
 foot_r {parent=calf_r, loc=<46.7401 2.5014 0.592>, rot=<0.0427 -0.0238 0.0353 -0.9982>}
 ball_r {parent=foot_r, loc=<3.5658 -10.8066 0.1544>, rot=<0.0001 -0.0001 -0.7063 -0.7079>}
 calf_twist_02_r {parent=calf_r, loc=<19.1546 1.7943 0.2638>, rot=<0.0237 0.0095 -0.0035 -0.9997>}
 calf_twist_01_r {parent=calf_r, loc=<35.8483 1.9197 0.5794>, rot=<0.0237 0.0095 -0.0035 -0.9997>}
 deform_knee_in_base_r {parent=calf_r, loc=<-0.8031 5.0231 -0.2457>, rot=<0 0 0 -1>}
 deform_knee_in_r {parent=deform_knee_in_base_r, loc=<0 0 0>, rot=<0 0 0 -1>}
 deform_knee_out_base_r {parent=calf_r, loc=<0 -4 0>, rot=<0 0 0 -1>}
 deform_knee_out_r {parent=deform_knee_out_base_r, loc=<0 0 0>, rot=<0 0 0 1>}
 thigh_twist_01_r {parent=thigh_r, loc=<27.952 0 0>, rot=<0.0013 0 0 -1>}
 deform_groin_base_r {parent=thigh_r, loc=<-0.4801 -5.6776 0.2486>, rot=<0 0 0 -1>}
 deform_groin_r {parent=deform_groin_base_r, loc=<0 0 0>, rot=<0 0 0 -1>}
 deform_zipper_r {parent=thigh_r, loc=<30.5805 -5.1273 2.311>, rot=<0.2356 -0.0064 -0.0417 -0.9709>}
 dyn_zipper_r {parent=deform_zipper_r, loc=<0 0 0>, rot=<0 0 0 1>}
 deform_glute_base_l {parent=pelvis, loc=<-7.3466 -9.0213 -9.1245>, rot=<0.0172 0.031 -0.4477 -0.8935>}
 deform_glute_l {parent=deform_glute_base_l, loc=<0 0 0>, rot=<0 0 0 1>}
 deform_glute_base_r {parent=pelvis, loc=<-7.3464 -9.0213 9.1245>, rot=<0.031 -0.0172 0.8935 -0.4477>}
 deform_glute_r {parent=deform_glute_base_r, loc=<0 0 0>, rot=<0 0 0 1>}
 deform_belt_l {parent=pelvis, loc=<0.7958 6.7749 -8.9241>, rot=<-0.0355 0.0009 0.9993 0.0155>}
 dyn_belt_l {parent=deform_belt_l, loc=<0 0 0>, rot=<0 0 0 1>}
 deform_belt_r {parent=pelvis, loc=<0.1311 6.7749 8.9241>, rot=<0.0355 -0.0009 0.9993 0.0155>}
 dyn_belt_r {parent=deform_belt_r, loc=<0 0 0>, rot=<0 0 0 1>}
 deform_belt_f {parent=pelvis, loc=<0.7399 8.1926 -0.0748>, rot=<-0.0089 0.0004 0.9998 0.0155>}
 dyn_belt_f {parent=deform_belt_f, loc=<0 0 0>, rot=<0 0 0 1>}
 deform_chain_f {parent=deform_belt_f, loc=<0.7915 -0.542 -3.0514>, rot=<-0.0566 -0.1594 -0.0745 -0.9828>}
 dyn_chain_f {parent=deform_chain_f, loc=<0 0 0>, rot=<0 0 0 -1>}
 dyn_chain_1_f {parent=dyn_chain_f, loc=<2.7492 -0.3879 -0.0827>, rot=<-0.0181 -0.1149 0.0356 -0.9926>}
 dyn_chain_2_f {parent=dyn_chain_1_f, loc=<6.6506 0 0>, rot=<-0.2235 -0.2378 -0.0683 -0.9428>}
 dyn_chain_3_f {parent=dyn_chain_2_f, loc=<6.8796 0 0>, rot=<-0.0652 -0.2504 -0.2608 -0.9301>}
 dyn_chain_4_f {parent=dyn_chain_3_f, loc=<7.2594 0 0>, rot=<-0.006 -0.0931 -0.1655 -0.9818>}
 deform_belt_gun_r {parent=pelvis, loc=<-1.3651 -0.4327 14.8028>, rot=<0.0141 -0.7072 0.7059 0.036>}
 dyn_belt_gun_r {parent=deform_belt_gun_r, loc=<0 0 0>, rot=<0 0 0 1>}
 deform_gun {parent=dyn_belt_gun_r, loc=<2.256 -1.63 0>, rot=<0 0 0 1>}
 dyn_gun {parent=deform_gun, loc=<0 0 0>, rot=<0 0 0 1>}
 dyn_gun_1 {parent=dyn_gun, loc=<21.5952 -2.2391 0.3895>, rot=<-0.0471 0.1207 -0.0207 -0.9914>}
 deform_chain_bk {parent=pelvis, loc=<0.0654 -12.0423 -6.9259>, rot=<0.0717 -0.9498 -0.1765 -0.248>}
 dyn_chain_bk {parent=deform_chain_bk, loc=<0 0 0>, rot=<0.0019 0.049 -0.0397 -0.998>}
 dyn_chain_1_bk {parent=dyn_chain_bk, loc=<6.9623 0 0>, rot=<0.0713 0.1469 -0.1989 -0.9663>}
 dyn_chain_2_bk {parent=dyn_chain_1_bk, loc=<6.9529 0 0>, rot=<0.1812 0.2049 -0.2297 -0.934>}
 dyn_chain_3_bk {parent=dyn_chain_2_bk, loc=<7.8926 0.0279 0.0522>, rot=<0.1809 -0.9675 0.0838 -0.1554>}
]
variables [
 Emotes {type=bool, default=False}
 Skydive {type=bool, default=False}
]
