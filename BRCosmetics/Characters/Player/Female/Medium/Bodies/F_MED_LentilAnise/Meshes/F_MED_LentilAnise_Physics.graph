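# Physics asset for the F_MED_LentilAnise player body cosmetic, covering its skirt or coat panels, collars, chains, belts and hip gun.
# Capsules on pelvis, spine, thighs and upper arms form the body, with small deform-bone spheres anchoring angle-limited panel chains and collars to the spine.
# Cross-constraints tie neighboring panels along the back hem and left side, and the front-left tip to the thigh.
# Chains, belts and gun hang from the pelvis.
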PhysicsAsset /BRCosmetics/Characters/Player/Female/Medium/Bodies/F_MED_LentilAnise/Meshes/F_MED_LentilAnise_Physics
bodies {
 deform_collar_l [{shapeType=Sphere, radius=1, shapeName=deform_collar_l_sphere}]
 deform_collar_r [{shapeType=Sphere, radius=1, shapeName=deform_collar_r_sphere}]
 deform_skirt_b_l [{shapeType=Sphere, radius=1, shapeName=deform_skirt_b_l_sphere}]
 deform_skirt_b_r [{shapeType=Sphere, radius=1, shapeName=deform_skirt_b_r_sphere}]
 deform_skirt_bk_l [{shapeType=Sphere, radius=1, shapeName=deform_skirt_bk_l_sphere}]
 deform_skirt_bk_r [{shapeType=Sphere, radius=1, shapeName=deform_skirt_bk_r_sphere}]
 deform_skirt_f_l [{shapeType=Sphere, radius=1, shapeName=deform_skirt_f_l_sphere}]
 deform_skirt_f_m_l [{shapeType=Sphere, radius=1, shapeName=deform_skirt_f_m_l_sphere}]
 deform_skirt_f_m_r [{shapeType=Sphere, radius=1, shapeName=deform_skirt_f_m_r_sphere}]
 deform_skirt_f_r [{shapeType=Sphere, radius=1, shapeName=deform_skirt_f_r_sphere}]
 deform_skirt_m_l [{shapeType=Sphere, radius=1, shapeName=deform_skirt_m_l_sphere}]
 deform_skirt_m_r [{shapeType=Sphere, radius=1, shapeName=deform_skirt_m_r_sphere}]
 dyn_belt_gun_r [{
   shapeType=Capsule
   length=11.7931
   radius=5.3399
   shapeName=dyn_belt_gun_r_capsule
   center=<8.9221 3.6792 1.6488>
   rotation=<rot -73.6128 0 0>
  }]
 dyn_belt_l [{
   shapeType=Capsule
   length=5.5567
   radius=5.7488
   shapeName=dyn_belt_l_capsule
   center=<-1.8142 3.9807 2.3791>
   rotation=<rot -12.1989 5.4292 -24.2178>
  }]
 dyn_chain_1_bk [{shapeType=Capsule, length=7.2816, radius=6.2412, shapeName=dyn_chain_1_bk_capsule, center=<2.9812 5.8678 0>, rotation=<rot 90 0 0>}]
 dyn_chain_2_f [{shapeType=Capsule, length=6.5984, radius=6.7305, shapeName=dyn_chain_2_f_capsule, center=<2.9812 5.6556 0>, rotation=<rot 90 0 0>}]
 dyn_chain_4_f [{shapeType=Capsule, length=6.0639, radius=7.2587, shapeName=dyn_chain_4_f_capsule, center=<0 6.0935 0>, rotation=<rot 90 0 0>}]
 dyn_collar_l [{shapeType=Capsule, length=8.0986, radius=2.4499, shapeName=dyn_collar_l_capsule, center=<9.0002 5.1498 0>, rotation=<rot 90 0 0>}]
 dyn_collar_r [{shapeType=Capsule, length=7.7464, radius=2.802, shapeName=dyn_collar_r_capsule, center=<8.3493 5.7238 0.5152>, rotation=<rot 90 180 0>}]
 dyn_gun [
  {
   shapeType=Capsule
   length=5.1064
   radius=11.6745
   shapeName=dyn_gun_capsule
   center=<12.85 -9.8793 0.283>
   rotation=<rot 87.5612 -176.2196 -176.9646>
  }
  {
   shapeType=Capsule
   length=18.8134
   radius=6.1134
   shapeName=dyn_gun_capsule
   center=<-0.0655 -5.1137 2.139>
   rotation=<rot 12.5393 0 -10.5379>
  }
 ]
 dyn_skirt_b_1_l [{shapeType=Capsule, length=6.7006, radius=7.7942, shapeName=dyn_skirt_b_1_l_capsule, center=<3.3767 -6.966 0>}]
 dyn_skirt_b_1_r [{shapeType=Capsule, length=10.0529, radius=4.5627, shapeName=dyn_skirt_b_1_r_capsule, center=<5.4814 -3.6444 -0.8166>, rotation=<rot 90 0 0>}]
 dyn_skirt_b_2_l [{shapeType=Capsule, length=6.7006, radius=7.7942, shapeName=dyn_skirt_b_2_l_capsule, center=<3.3767 -6.966 0>}]
 dyn_skirt_b_2_r [{shapeType=Capsule, length=7.7324, radius=4.5627, shapeName=dyn_skirt_b_2_r_capsule, center=<3.3676 -3.6444 0>, rotation=<rot 90 0 0>}]
 dyn_skirt_b_l [{shapeType=Capsule, length=6.7006, radius=7.7942, shapeName=dyn_skirt_b_l_capsule, center=<3.3767 -5.6942 0>}]
 dyn_skirt_b_r [{shapeType=Capsule, length=9.932, radius=4.5627, shapeName=dyn_skirt_b_r_capsule, center=<5.4814 -3.6444 0>, rotation=<rot 90 0 0>}]
 dyn_skirt_bk_1_l [{shapeType=Capsule, length=12.2058, radius=7.7942, shapeName=dyn_skirt_bk_1_l_capsule, center=<3.3767 -5.6942 0>}]
 dyn_skirt_bk_1_r [{shapeType=Capsule, length=12.2058, radius=7.7942, shapeName=dyn_skirt_bk_1_r_capsule, center=<3.3767 -5.6942 0>}]
 dyn_skirt_bk_2_l [{shapeType=Capsule, length=12.2058, radius=7.7942, shapeName=dyn_skirt_bk_2_l_capsule, center=<3.3767 -5.6942 0>}]
 dyn_skirt_bk_2_r [{shapeType=Capsule, length=12.2058, radius=7.7942, shapeName=dyn_skirt_bk_2_r_capsule, center=<3.3767 -5.6942 0>}]
 dyn_skirt_bk_l [{shapeType=Capsule, length=12.2058, radius=7.7942, shapeName=dyn_skirt_bk_l_capsule, center=<3.3767 -7.5399 0>}]
 dyn_skirt_bk_r [{shapeType=Capsule, length=12.2058, radius=7.7942, shapeName=dyn_skirt_bk_r_capsule, center=<3.3767 -7.5399 0>}]
 dyn_skirt_f_1_l [{shapeType=Capsule, length=9.1569, radius=4, shapeName=dyn_skirt_f_1_l_capsule, center=<4.3048 -2.8301 -1.3653>, rotation=<rot 90 0 0>}]
 dyn_skirt_f_1_r [{shapeType=Capsule, length=7.3003, radius=5, shapeName=dyn_skirt_f_1_r_capsule, center=<3.3767 -2.8969 0>}]
 dyn_skirt_f_2_l [{shapeType=Capsule, length=8.5728, radius=4, shapeName=dyn_skirt_f_2_l_capsule, center=<4.3048 -2.8301 -1.3653>, rotation=<rot 90 0 0>}]
 dyn_skirt_f_3_l [{shapeType=Capsule, length=6, radius=4, shapeName=dyn_skirt_f_3_l_capsule, center=<4.3048 -2.8301 -1.3653>, rotation=<rot 90 0 0>}]
 dyn_skirt_f_l [{shapeType=Capsule, length=9.3155, radius=4, shapeName=dyn_skirt_f_l_capsule, center=<4.3048 -2.8301 -1.3653>, rotation=<rot 90 0 0>}]
 dyn_skirt_f_m_1_l [{shapeType=Capsule, length=8.5598, radius=4, shapeName=dyn_skirt_f_m_1_l_capsule, center=<4.5095 -2.4897 0>, rotation=<rot 90 0 0>}]
 dyn_skirt_f_m_1_r [{shapeType=Capsule, length=7.3003, radius=5, shapeName=dyn_skirt_f_m_1_r_capsule, center=<3.3767 -2.8969 0>}]
 dyn_skirt_f_m_2_l [{shapeType=Capsule, length=9.5403, radius=4, shapeName=dyn_skirt_f_m_2_l_capsule, center=<4.9175 -2.4897 0>, rotation=<rot 90 0 0>}]
 dyn_skirt_f_m_2_r [{shapeType=Capsule, length=7.3003, radius=5, shapeName=dyn_skirt_f_m_2_r_capsule, center=<3.3767 -2.8969 0>}]
 dyn_skirt_f_m_3_l [{shapeType=Capsule, length=8.5954, radius=4, shapeName=dyn_skirt_f_m_3_l_capsule, center=<4.3944 -2.5263 0.1403>, rotation=<rot 25.4976 0 0>}]
 dyn_skirt_f_m_l [{shapeType=Capsule, length=6, radius=4, shapeName=dyn_skirt_f_m_l_capsule, center=<3.078 -2.4897 0>, rotation=<rot 90 0 0>}]
 dyn_skirt_f_m_r [{shapeType=Sphere, radius=1, shapeName=dyn_skirt_f_m_r_sphere}]
 dyn_skirt_f_r [{shapeType=Sphere, radius=1, shapeName=dyn_skirt_f_r_sphere}]
 dyn_skirt_m_1_l [{shapeType=Capsule, length=10, radius=5, shapeName=dyn_skirt_m_1_l_capsule, center=<4.4802 -3.792 0>, rotation=<rot 90 0 0>}]
 dyn_skirt_m_1_r [{shapeType=Capsule, length=5.5802, radius=3.2534, shapeName=dyn_skirt_m_1_r_capsule, center=<1.5428 -0.0836 0>, rotation=<rot 90 9.3661 0>}]
 dyn_skirt_m_2_l [{shapeType=Capsule, length=10, radius=5, shapeName=dyn_skirt_m_2_l_capsule, center=<7.8724 -3.792 0>, rotation=<rot 27.984 0 0>}]
 dyn_skirt_m_2_r [{shapeType=Capsule, length=10.4643, radius=4.0304, shapeName=dyn_skirt_m_2_r_capsule, center=<5.7896 -2.4208 0>, rotation=<rot 90 0 0>}]
 dyn_skirt_m_l [{shapeType=Capsule, length=10, radius=5, shapeName=dyn_skirt_m_l_capsule, center=<4.4802 -3.792 0>, rotation=<rot 90 0 0>}]
 dyn_skirt_m_r [{shapeType=Capsule, length=7.7494, radius=3.2494, shapeName=dyn_skirt_m_r_capsule, center=<4.1172 -2.4208 0>, rotation=<rot 90 0 0>}]
 pelvis [{shapeType=Capsule, length=8.2395, radius=11.119, shapeName=pelvis_capsule, center=<0 -2.3941 0>}]
 spine_01 [{shapeType=Capsule, length=5.9404, radius=9.9044, shapeName=spine_01_capsule}]
 spine_02 [{shapeType=Capsule, length=0.3813, radius=9.9712, shapeName=spine_02_capsule}]
 spine_03 [{shapeType=Capsule, length=4.1018, radius=8.8461, shapeName=spine_03_capsule, center=<0 1.4981 0>}]
 spine_04 [{shapeType=Capsule, length=3.5613, radius=11.5058, shapeName=spine_04_capsule, center=<0 2.086 0>}]
 thigh_l [{shapeType=Capsule, length=40.9701, radius=9.0141, shapeName=thigh_l_capsule, center=<-22.5849 -2.1444 0.5921>, rotation=<rot 90 -178.9404 0>}]
 thigh_r [{shapeType=Capsule, length=40.9701, radius=9.0141, shapeName=thigh_r_capsule, center=<22.6102 2.071 -2.8115>, rotation=<rot 90 0 0>}]
 upperarm_l [{shapeType=Capsule, length=24.1626, radius=4.7341, shapeName=upperarm_l_capsule, center=<12.5821 0 0.479>, rotation=<rot 90 180 0>}]
 upperarm_r [{shapeType=Capsule, length=24.2808, radius=4.6159, shapeName=upperarm_r_capsule, center=<-16.1315 1.9064 0>, rotation=<rot 90 0 0>}]
}
constraints [
 {bone1Name=dyn_chain_2_f, bone2Name=pelvis, swing1LimitDegrees=15, swing2LimitDegrees=15, twistLimitDegrees=15}
 {
  bone1Name=dyn_chain_4_f
  bone2Name=dyn_chain_2_f
  swing1LimitDegrees=15
  swing1Motion=Locked
  swing2LimitDegrees=15
  swing2Motion=Locked
  twistLimitDegrees=15
  twistMotion=Locked
 }
 {bone1Name=dyn_chain_1_bk, bone2Name=pelvis, swing1LimitDegrees=15, swing2LimitDegrees=15, twistLimitDegrees=15}
 {
  bone1Name=spine_01
  bone2Name=pelvis
  swing1LimitDegrees=45
  swing1Motion=Free
  swing2LimitDegrees=45
  swing2Motion=Free
  twistLimitDegrees=45
  twistMotion=Free
 }
 {
  bone1Name=spine_02
  bone2Name=spine_01
  swing1LimitDegrees=45
  swing1Motion=Free
  swing2LimitDegrees=45
  swing2Motion=Free
  twistLimitDegrees=45
  twistMotion=Free
 }
 {
  bone1Name=deform_skirt_b_l
  bone2Name=spine_02
  swing1LimitDegrees=45
  swing1Motion=Free
  swing2LimitDegrees=45
  swing2Motion=Free
  twistLimitDegrees=45
  twistMotion=Free
 }
 {bone1Name=dyn_skirt_b_l, bone2Name=deform_skirt_b_l, swing1LimitDegrees=30, swing2LimitDegrees=15, twistLimitDegrees=5}
 {bone1Name=dyn_skirt_b_1_l, bone2Name=dyn_skirt_b_l, swing1LimitDegrees=30, swing2LimitDegrees=15, twistLimitDegrees=5}
 {bone1Name=dyn_skirt_b_2_l, bone2Name=dyn_skirt_b_1_l, swing1LimitDegrees=30, swing2LimitDegrees=15, twistLimitDegrees=5}
 {
  bone1Name=deform_skirt_b_r
  bone2Name=spine_02
  swing1LimitDegrees=45
  swing1Motion=Free
  swing2LimitDegrees=45
  swing2Motion=Free
  twistLimitDegrees=45
  twistMotion=Free
 }
 {bone1Name=dyn_skirt_b_r, bone2Name=deform_skirt_b_r, swing1LimitDegrees=30, swing2LimitDegrees=15, twistLimitDegrees=5}
 {bone1Name=dyn_skirt_b_1_r, bone2Name=dyn_skirt_b_r, swing1LimitDegrees=30, swing2LimitDegrees=15, twistLimitDegrees=5}
 {bone1Name=dyn_skirt_b_2_r, bone2Name=dyn_skirt_b_1_r, swing1LimitDegrees=30, swing2LimitDegrees=15, twistLimitDegrees=5}
 {
  bone1Name=deform_skirt_bk_l
  bone2Name=spine_02
  swing1LimitDegrees=45
  swing1Motion=Free
  swing2LimitDegrees=45
  swing2Motion=Free
  twistLimitDegrees=45
  twistMotion=Free
 }
 {bone1Name=dyn_skirt_bk_l, bone2Name=deform_skirt_bk_l, swing1LimitDegrees=30, swing2LimitDegrees=15, twistLimitDegrees=5}
 {bone1Name=dyn_skirt_bk_1_l, bone2Name=dyn_skirt_bk_l, swing1LimitDegrees=30, swing2LimitDegrees=15, twistLimitDegrees=5}
 {bone1Name=dyn_skirt_bk_2_l, bone2Name=dyn_skirt_bk_1_l, swing1LimitDegrees=30, swing2LimitDegrees=15, twistLimitDegrees=5}
 {
  bone1Name=deform_skirt_bk_r
  bone2Name=spine_02
  swing1LimitDegrees=45
  swing1Motion=Free
  swing2LimitDegrees=45
  swing2Motion=Free
  twistLimitDegrees=45
  twistMotion=Free
 }
 {bone1Name=dyn_skirt_bk_r, bone2Name=deform_skirt_bk_r, swing1LimitDegrees=30, swing2LimitDegrees=15, twistLimitDegrees=5}
 {bone1Name=dyn_skirt_bk_1_r, bone2Name=dyn_skirt_bk_r, swing1LimitDegrees=30, swing2LimitDegrees=15, twistLimitDegrees=5}
 {bone1Name=dyn_skirt_bk_2_r, bone2Name=dyn_skirt_bk_1_r, swing1LimitDegrees=30, swing2LimitDegrees=15, twistLimitDegrees=5}
 {
  bone1Name=spine_03
  bone2Name=spine_02
  swing1LimitDegrees=45
  swing1Motion=Free
  swing2LimitDegrees=45
  swing2Motion=Free
  twistLimitDegrees=45
  twistMotion=Free
 }
 {
  bone1Name=deform_skirt_m_l
  bone2Name=spine_03
  swing1LimitDegrees=45
  swing1Motion=Free
  swing2LimitDegrees=45
  swing2Motion=Free
  twistLimitDegrees=45
  twistMotion=Free
 }
 {bone1Name=dyn_skirt_m_l, bone2Name=deform_skirt_m_l, swing1LimitDegrees=30, swing2LimitDegrees=15, twistLimitDegrees=5}
 {bone1Name=dyn_skirt_m_1_l, bone2Name=dyn_skirt_m_l, swing1LimitDegrees=30, swing2LimitDegrees=15, twistLimitDegrees=5}
 {bone1Name=dyn_skirt_m_2_l, bone2Name=dyn_skirt_m_1_l, swing1LimitDegrees=30, swing2LimitDegrees=15, twistLimitDegrees=5}
 {
  bone1Name=deform_skirt_m_r
  bone2Name=spine_03
  swing1LimitDegrees=45
  swing1Motion=Free
  swing2LimitDegrees=45
  swing2Motion=Free
  twistLimitDegrees=45
  twistMotion=Free
 }
 {bone1Name=dyn_skirt_m_r, bone2Name=deform_skirt_m_r, swing1LimitDegrees=30, swing2LimitDegrees=15, twistLimitDegrees=5}
 {bone1Name=dyn_skirt_m_1_r, bone2Name=dyn_skirt_m_r, swing1LimitDegrees=30, swing2LimitDegrees=15, twistLimitDegrees=5}
 {bone1Name=dyn_skirt_m_2_r, bone2Name=dyn_skirt_m_1_r, swing1LimitDegrees=30, swing2LimitDegrees=15, twistLimitDegrees=5}
 {
  bone1Name=spine_04
  bone2Name=spine_03
  swing1LimitDegrees=45
  swing1Motion=Free
  swing2LimitDegrees=45
  swing2Motion=Free
  twistLimitDegrees=45
  twistMotion=Free
 }
 {
  bone1Name=deform_skirt_f_l
  bone2Name=spine_04
  swing1LimitDegrees=45
  swing1Motion=Free
  swing2LimitDegrees=45
  swing2Motion=Free
  twistLimitDegrees=45
  twistMotion=Free
 }
 {bone1Name=dyn_skirt_f_l, bone2Name=deform_skirt_f_l, swing1LimitDegrees=15, swing2LimitDegrees=15, twistLimitDegrees=1}
 {bone1Name=dyn_skirt_f_1_l, bone2Name=dyn_skirt_f_l, swing1LimitDegrees=45, swing2LimitDegrees=15, twistLimitDegrees=5}
 {bone1Name=dyn_skirt_f_2_l, bone2Name=dyn_skirt_f_1_l, swing1LimitDegrees=30, swing2LimitDegrees=15, twistLimitDegrees=5}
 {bone1Name=dyn_skirt_f_3_l, bone2Name=dyn_skirt_f_2_l, swing1LimitDegrees=30, swing2LimitDegrees=15, twistLimitDegrees=5}
 {
  bone1Name=deform_skirt_f_m_l
  bone2Name=spine_04
  swing1LimitDegrees=45
  swing1Motion=Free
  swing2LimitDegrees=45
  swing2Motion=Free
  twistLimitDegrees=45
  twistMotion=Free
 }
 {bone1Name=dyn_skirt_f_m_l, bone2Name=deform_skirt_f_m_l, swing1LimitDegrees=15, swing2LimitDegrees=10, twistLimitDegrees=1}
 {bone1Name=dyn_skirt_f_m_1_l, bone2Name=dyn_skirt_f_m_l, swing1LimitDegrees=45, swing2LimitDegrees=15, twistLimitDegrees=5}
 {bone1Name=dyn_skirt_f_m_2_l, bone2Name=dyn_skirt_f_m_1_l, swing1LimitDegrees=30, swing2LimitDegrees=15, twistLimitDegrees=5}
 {bone1Name=dyn_skirt_f_m_3_l, bone2Name=dyn_skirt_f_m_2_l, swing1LimitDegrees=30, swing2LimitDegrees=15, twistLimitDegrees=5}
 {
  bone1Name=deform_skirt_f_m_r
  bone2Name=spine_04
  swing1LimitDegrees=45
  swing1Motion=Free
  swing2LimitDegrees=45
  swing2Motion=Free
  twistLimitDegrees=45
  twistMotion=Free
 }
 {bone1Name=dyn_skirt_f_m_r, bone2Name=deform_skirt_f_m_r, swing1LimitDegrees=30, swing2LimitDegrees=15, twistLimitDegrees=5}
 {bone1Name=dyn_skirt_f_m_1_r, bone2Name=dyn_skirt_f_m_r, swing1LimitDegrees=30, swing2LimitDegrees=15, twistLimitDegrees=5}
 {bone1Name=dyn_skirt_f_m_2_r, bone2Name=dyn_skirt_f_m_1_r, swing1LimitDegrees=30, swing2LimitDegrees=15, twistLimitDegrees=5}
 {
  bone1Name=deform_skirt_f_r
  bone2Name=spine_04
  swing1LimitDegrees=45
  swing1Motion=Free
  swing2LimitDegrees=45
  swing2Motion=Free
  twistLimitDegrees=45
  twistMotion=Free
 }
 {bone1Name=dyn_skirt_f_r, bone2Name=deform_skirt_f_r, swing1LimitDegrees=30, swing2LimitDegrees=15, twistLimitDegrees=5}
 {bone1Name=dyn_skirt_f_1_r, bone2Name=dyn_skirt_f_r, swing1LimitDegrees=30, swing2LimitDegrees=15, twistLimitDegrees=5}
 {
  bone1Name=deform_collar_l
  bone2Name=spine_04
  swing1LimitDegrees=45
  swing1Motion=Free
  swing2LimitDegrees=45
  swing2Motion=Free
  twistLimitDegrees=45
  twistMotion=Free
 }
 {bone1Name=dyn_collar_l, bone2Name=deform_collar_l, swing1LimitDegrees=15, swing2LimitDegrees=15, twistLimitDegrees=5}
 {
  bone1Name=upperarm_l
  bone2Name=spine_04
  swing1LimitDegrees=45
  swing1Motion=Free
  swing2LimitDegrees=45
  swing2Motion=Free
  twistLimitDegrees=45
  twistMotion=Free
 }
 {
  bone1Name=deform_collar_r
  bone2Name=spine_04
  swing1LimitDegrees=45
  swing1Motion=Free
  swing2LimitDegrees=45
  swing2Motion=Free
  twistLimitDegrees=45
  twistMotion=Free
 }
 {bone1Name=dyn_collar_r, bone2Name=deform_collar_r, swing1LimitDegrees=15, swing2LimitDegrees=15, twistLimitDegrees=5}
 {
  bone1Name=upperarm_r
  bone2Name=spine_04
  swing1LimitDegrees=45
  swing1Motion=Free
  swing2LimitDegrees=45
  swing2Motion=Free
  twistLimitDegrees=45
  twistMotion=Free
 }
 {
  bone1Name=thigh_l
  bone2Name=pelvis
  swing1LimitDegrees=45
  swing1Motion=Free
  swing2LimitDegrees=45
  swing2Motion=Free
  twistLimitDegrees=45
  twistMotion=Free
 }
 {
  bone1Name=thigh_r
  bone2Name=pelvis
  swing1LimitDegrees=45
  swing1Motion=Free
  swing2LimitDegrees=45
  swing2Motion=Free
  twistLimitDegrees=45
  twistMotion=Free
 }
 {bone1Name=dyn_skirt_m_2_l, bone2Name=dyn_skirt_b_2_l, swing1LimitDegrees=15, swing2LimitDegrees=15, twistLimitDegrees=5}
 {bone1Name=dyn_skirt_b_2_l, bone2Name=dyn_skirt_bk_2_l, swing1LimitDegrees=15, swing2LimitDegrees=15, twistLimitDegrees=5}
 {bone1Name=dyn_skirt_bk_2_l, bone2Name=dyn_skirt_bk_2_r, swing1LimitDegrees=15, swing2LimitDegrees=15, twistLimitDegrees=5}
 {bone1Name=dyn_skirt_bk_2_r, bone2Name=dyn_skirt_b_2_r, swing1LimitDegrees=15, swing2LimitDegrees=15, twistLimitDegrees=5}
 {bone1Name=dyn_skirt_b_2_r, bone2Name=dyn_skirt_m_2_r, swing1LimitDegrees=15, swing2LimitDegrees=15, twistLimitDegrees=5}
 {
  bone1Name=dyn_skirt_f_3_l
  bone2Name=thigh_l
  swing1LimitDegrees=45
  swing1Motion=Free
  swing2LimitDegrees=45
  swing2Motion=Free
  twistLimitDegrees=45
  twistMotion=Free
 }
 {
  bone1Name=dyn_belt_gun_r
  bone2Name=pelvis
  swing1LimitDegrees=45
  swing1Motion=Free
  swing2LimitDegrees=45
  swing2Motion=Free
  twistLimitDegrees=45
  twistMotion=Free
 }
 {
  bone1Name=dyn_belt_l
  bone2Name=pelvis
  swing1LimitDegrees=45
  swing1Motion=Free
  swing2LimitDegrees=45
  swing2Motion=Free
  twistLimitDegrees=45
  twistMotion=Free
 }
 {
  bone1Name=dyn_gun
  bone2Name=dyn_belt_gun_r
  swing1LimitDegrees=45
  swing1Motion=Free
  swing2LimitDegrees=45
  swing2Motion=Free
  twistLimitDegrees=45
  twistMotion=Free
 }
 {bone1Name=dyn_skirt_f_m_l, bone2Name=dyn_skirt_f_l, swing1LimitDegrees=15, swing2LimitDegrees=15, twistLimitDegrees=15}
 {bone1Name=dyn_skirt_f_m_1_l, bone2Name=dyn_skirt_f_1_l, swing1LimitDegrees=15, swing2LimitDegrees=15, twistLimitDegrees=15}
 {bone1Name=dyn_skirt_f_m_2_l, bone2Name=dyn_skirt_f_2_l, swing1LimitDegrees=15, swing2LimitDegrees=15, twistLimitDegrees=15}
 {bone1Name=dyn_skirt_f_m_3_l, bone2Name=dyn_skirt_f_3_l, swing1LimitDegrees=15, swing2LimitDegrees=15, twistLimitDegrees=15}
 {
  bone1Name=dyn_skirt_m_l
  bone2Name=dyn_skirt_f_m_1_l
  swing1LimitDegrees=15
  swing1Motion=Free
  swing2LimitDegrees=15
  swing2Motion=Free
  twistLimitDegrees=15
  twistMotion=Free
 }
 {bone1Name=dyn_skirt_m_1_l, bone2Name=dyn_skirt_f_m_2_l, swing1LimitDegrees=15, swing2LimitDegrees=15, twistLimitDegrees=15}
 {bone1Name=dyn_skirt_m_2_l, bone2Name=dyn_skirt_f_m_3_l, swing1LimitDegrees=15, swing2LimitDegrees=15, twistLimitDegrees=15}
 {
  bone1Name=dyn_skirt_b_1_r
  bone2Name=dyn_skirt_m_1_r
  swing1LimitDegrees=15
  swing1Motion=Free
  swing2LimitDegrees=15
  swing2Motion=Free
  twistLimitDegrees=5
  twistMotion=Free
 }
]
