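# Physics asset for the LentilAnise female medium cosmetic body, defining a chain that hangs at the left hip.
# Front and back capsule strands descend from pelvis via small sphere anchor bodies, with angular-limited upper joints and locked end joints.
# A locked constraint ties the last front link to the back strand, closing the loop.
# Left thigh capsules provide leg shapes, and an angular-free constraint also links the front strand to the thigh.
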PhysicsAsset /BRCosmetics/Characters/Player/Female/Medium/Bodies/F_MED_LentilAnise/Meshes/F_MED_LentilAnise_Physics1
bodies {
 deform_belt_f [{shapeType=Sphere, radius=1, shapeName=deform_belt_f_sphere}]
 deform_chain_bk [{shapeType=Sphere, radius=1, shapeName=deform_chain_bk_sphere}]
 deform_chain_f [{shapeType=Sphere, radius=1, shapeName=deform_chain_f_sphere}]
 dyn_chain_1_bk [{shapeType=Capsule, length=8, radius=4, shapeName=dyn_chain_1_bk_capsule, center=<4.58 -2.9663 0>, rotation=<rot 90 0 0>}]
 dyn_chain_1_f [{shapeType=Capsule, length=6.5068, radius=3.1745, shapeName=dyn_chain_1_f_capsule, center=<3.1719 -3.0883 0>, rotation=<rot 90 0 0>}]
 dyn_chain_2_bk [{shapeType=Capsule, length=8, radius=4, shapeName=dyn_chain_2_bk_capsule, center=<4.58 -2.9663 0>, rotation=<rot 90 0 0>}]
 dyn_chain_2_f [{shapeType=Capsule, length=8, radius=4, shapeName=dyn_chain_2_f_capsule, center=<3.5376 -2.9663 0>, rotation=<rot 90 0 0>}]
 dyn_chain_3_bk [{shapeType=Capsule, length=8, radius=4, shapeName=dyn_chain_3_bk_capsule, center=<0 -2.9663 0>, rotation=<rot 90 0 0>}]
 dyn_chain_3_f [{shapeType=Capsule, length=8, radius=4, shapeName=dyn_chain_3_f_capsule, center=<3.5376 -2.9663 0>, rotation=<rot 90 0 0>}]
 dyn_chain_4_f [{shapeType=Capsule, length=8, radius=4, shapeName=dyn_chain_4_f_capsule, center=<-3.2693 -2.9663 0>, rotation=<rot 90 0 0>}]
 dyn_chain_bk [{shapeType=Capsule, length=8, radius=4, shapeName=dyn_chain_bk_capsule, center=<4.58 -2.9663 0>, rotation=<rot 90 0 0>}]
 dyn_chain_f [{shapeType=Sphere, radius=1, shapeName=dyn_chain_f_sphere}]
 pelvis [{shapeType=Sphere, radius=1, shapeName=pelvis_sphere}]
 thigh_l [{shapeType=Capsule, length=43.3298, radius=6.6544, shapeName=thigh_l_capsule, center=<-22.5923 -0.5031 0.953>, rotation=<rot 90 178.2372 0>}]
 thigh_twist_01_l [{
   shapeType=Capsule
   length=19.6979
   radius=6.954
   shapeName=thigh_twist_01_l_capsule
   center=<11.7425 -1.8464 0.9999>
   rotation=<rot 90 -8.8162 0>
  }]
}
constraints [
 {
  bone1Name=deform_chain_f
  bone2Name=deform_belt_f
  swing1LimitDegrees=45
  swing1Motion=Free
  swing2LimitDegrees=45
  swing2Motion=Free
  twistLimitDegrees=45
  twistMotion=Free
 }
 {bone1Name=dyn_chain_f, bone2Name=deform_chain_f, swing1LimitDegrees=15, swing2LimitDegrees=15, twistLimitDegrees=15}
 {bone1Name=dyn_chain_1_f, bone2Name=dyn_chain_f, swing1LimitDegrees=60, swing2LimitDegrees=35, twistLimitDegrees=15}
 {bone1Name=dyn_chain_2_f, bone2Name=dyn_chain_1_f, swing1LimitDegrees=15, swing2LimitDegrees=15, twistLimitDegrees=15}
 {bone1Name=dyn_chain_3_f, bone2Name=dyn_chain_2_f, swing1LimitDegrees=15, swing2LimitDegrees=15, twistLimitDegrees=15}
 {
  bone1Name=dyn_chain_4_f
  bone2Name=dyn_chain_3_f
  swing1LimitDegrees=15
  swing1Motion=Locked
  swing2LimitDegrees=15
  swing2Motion=Locked
  twistLimitDegrees=15
  twistMotion=Locked
 }
 {
  bone1Name=deform_belt_f
  bone2Name=pelvis
  swing1LimitDegrees=45
  swing1Motion=Free
  swing2LimitDegrees=45
  swing2Motion=Free
  twistLimitDegrees=45
  twistMotion=Free
 }
 {
  bone1Name=deform_chain_bk
  bone2Name=pelvis
  swing1LimitDegrees=45
  swing1Motion=Free
  swing2LimitDegrees=45
  swing2Motion=Free
  twistLimitDegrees=45
  twistMotion=Free
 }
 {bone1Name=dyn_chain_bk, bone2Name=deform_chain_bk, swing1LimitDegrees=15, swing2LimitDegrees=15, twistLimitDegrees=15}
 {bone1Name=dyn_chain_1_bk, bone2Name=dyn_chain_bk, swing1LimitDegrees=15, swing2LimitDegrees=15, twistLimitDegrees=15}
 {
  bone1Name=dyn_chain_2_bk
  bone2Name=dyn_chain_1_bk
  swing1LimitDegrees=15
  swing1Motion=Locked
  swing2LimitDegrees=15
  swing2Motion=Locked
  twistLimitDegrees=15
  twistMotion=Locked
 }
 {
  bone1Name=dyn_chain_3_bk
  bone2Name=dyn_chain_2_bk
  swing1LimitDegrees=15
  swing1Motion=Locked
  swing2LimitDegrees=15
  swing2Motion=Locked
  twistLimitDegrees=15
  twistMotion=Locked
 }
 {
  bone1Name=thigh_l
  bone2Name=pelvis
  swing1LimitDegrees=45
  swing1Motion=Free
  swing2LimitDegrees=45
  swing2Motion=Free
  twistLimitDegrees=45
  twistMotion=Free
 }
 {
  bone1Name=dyn_chain_4_f
  bone2Name=dyn_chain_2_bk
  swing1LimitDegrees=10
  swing1Motion=Locked
  swing2LimitDegrees=10
  swing2Motion=Locked
  twistLimitDegrees=10
  twistMotion=Locked
 }
 {
  bone1Name=thigh_twist_01_l
  bone2Name=thigh_l
  swing1LimitDegrees=45
  swing1Motion=Free
  swing2LimitDegrees=45
  swing2Motion=Free
  twistLimitDegrees=45
  twistMotion=Free
 }
 {
  bone1Name=dyn_chain_2_f
  bone2Name=thigh_l
  swing1LimitDegrees=45
  swing1Motion=Free
  swing2LimitDegrees=45
  swing2Motion=Free
  twistLimitDegrees=45
  twistMotion=Free
 }
]
