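# Post-process style cosmetic AnimBP for the LentilAnise female body, adding secondary motion to the attached parent's copied pose.
# A Control Rig receives emote and skydive flags, then feeds RigidBody physics, spring AnimDynamics on zipper and gun bones, and additive skirt and chain corrections for sliding and stationary crouching.
# Update borrows GnatGala parameters to set Gravity Override, but physics uses fixed gravity.
# A second RigidBody node is unconnected.
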AnimBlueprint /BRCosmetics/Characters/Player/Female/Medium/Bodies/F_MED_LentilAnise/Meshes/F_MED_LentilAnise_AnimBP
target_skeleton=&/F_MED_LentilAnise_Skeleton
graphs [
 AnimGraph {nodes [
   RigidBody!(ComponentPose, Alpha, OverrideWorldGravity:Vector="0,0,-980") -> (Pose) {
    bClampLinearTranslationLimitToRefPose=true
    bOverrideWorldGravity=true
    overridePhysicsAsset=&/F_MED_LentilAnise_Physics1
    simulationSpace=BaseBoneSpace
    baseBoneRef {boneName=pelvis}
    overrideWorldGravity {z=-980}
    simSpaceSettings {
     maxAngularAcceleration=2000
     maxAngularVelocity=6000
     maxLinearAcceleration=2500
     maxLinearVelocity=6000
     velocityScaleZ=0.6
     worldAlpha=0.6
     externalLinearDragV=<1 1 1>
    }
   }
   n1 = Control Rig!(Source=Copy Pose From Mesh() {bUseAttachedParent=true}, Alpha, Emotes:bool=Get bIsPlayingEmote(), Skydive:bool=Get bIsSkyDiving()) -> (Pose) {controlRigAssetReference {blueprintRigClass=&/F_MED_LentilAnise_CtrlRig.F_MED_LentilAnise_CtrlRig_C}}
   n2 = RigidBody(ComponentPose=Local To Component(LocalPose=@n1), OverrideWorldGravity="0,0,-980") {
    bClampLinearTranslationLimitToRefPose=true
    bOverrideWorldGravity=true
    overridePhysicsAsset=&/F_MED_LentilAnise_Physics
    simulationSpace=BaseBoneSpace
    baseBoneRef {boneName=spine_02}
    simSpaceSettings {
     maxAngularAcceleration=2000
     maxAngularVelocity=6000
     maxLinearAcceleration=2500
     maxLinearVelocity=6000
     velocityScaleZ=0.6
     worldAlpha=0.6
     externalLinearDragV=<1 1 1>
    }
   }
   n3 = Anim Dynamics - Bone: dyn_zipper_l!(ComponentPose=@n2, Alpha=1) -> (Pose) {
    angularDampingOverride=0.999
    angularSpringConstant=1
    bAngularSpring=true
    bOverrideAngularDamping=true
    simulationSpace=RootRelative
    boundBone {boneName=dyn_zipper_l}
    physicsBodyDefinitions [{
      boundBone {boneName=dyn_zipper_l}
      boxExtents=<10 3 5>
      constraintSetup {angularLimitsMax=<0 35 0>, angularLimitsMin=<0 -35 -90>, angularTarget=<0 0 0>}
      localJointOffset=<0.3 0 0>
     }]
   }
   n4 = Anim Dynamics - Bone: dyn_zipper_r!(ComponentPose=@n3, Alpha=1) -> (Pose) {
    angularDampingOverride=0.999
    angularSpringConstant=1
    bAngularSpring=true
    bOverrideAngularDamping=true
    simulationSpace=RootRelative
    boundBone {boneName=dyn_zipper_r}
    physicsBodyDefinitions [{
      boundBone {boneName=dyn_zipper_r}
      boxExtents=<10 3 5>
      constraintSetup {angularLimitsMax=<0 35 0>, angularLimitsMin=<0 -35 -90>, angularTarget=<0 0 0>}
      localJointOffset=<0.3 0 0>
     }]
   }
   n5 = Anim Dynamics - Bone: dyn_gun!(ComponentPose=@n4, Alpha=0.3) -> (Pose) {
    angularDampingOverride=0.999
    angularSpringConstant=500
    bAngularSpring=true
    bOverrideAngularDamping=true
    simulationSpace=RootRelative
    boundBone {boneName=dyn_gun}
    physicsBodyDefinitions [{
      boundBone {boneName=dyn_gun}
      boxExtents=<10 3 5>
      constraintSetup {angularLimitsMax=<0 35 0>, angularLimitsMin=<0 -35 0>, angularTarget=<1 0 0>}
      localJointOffset=<0.3 0 0>
     }]
   }
   n6 = Anim Dynamics - Bone: dyn_gun_1!(ComponentPose=@n5, Alpha=0.5) -> (Pose) {
    angularDampingOverride=0.9999
    angularSpringConstant=500
    bAngularSpring=true
    bLinearSpring=true
    bOverrideAngularDamping=true
    bOverrideLinearDamping=true
    linearDampingOverride=0.9999
    linearSpringConstant=50
    simulationSpace=BoneRelative
    boundBone {boneName=dyn_gun_1}
    physicsBodyDefinitions [{
      boundBone {boneName=dyn_gun_1}
      boxExtents=<15 7 15>
      constraintSetup {angularLimitsMax=<0 2 0>, angularLimitsMin=<0 0 0>, angularTarget=<1 0 0>}
      localJointOffset=<0.3 0 0>
     }]
    relativeSpaceBone {boneName=deform_gun}
   }
   n7 = "Transform (Modify) Bone - Bone: deform_skirt_f_l"!(ComponentPose=@n6, bAlphaBoolEnabled:bool=Get bIsSliding(), Translation="0,0,0", Rotation="0.557,-19.888,-8.474", Scale="1,1,1") -> (Pose) {
    alphaInputType=Bool
    rotationMode=BMM_Additive
    rotationSpace=BCS_BoneSpace
    alphaBoolBlend {blendInTime=0.1, blendOutTime=0.5}
    boneToModify {boneName=deform_skirt_f_l}
    rotation=<rot 0.557 -19.888 -8.474>
   }
   n8 = "Transform (Modify) Bone - Bone: dyn_chain_1_f"!(
    ComponentPose=@n7
    bAlphaBoolEnabled:bool=Get CrouchingNotMoving() -> (CrouchingNotMoving:bool)
    Translation="0,0,0"
    Rotation="10.852,3.699,1.272"
    Scale="1,1,1"
   ) -> (Pose) {
    alphaInputType=Bool
    rotationMode=BMM_Additive
    rotationSpace=BCS_BoneSpace
    alphaBoolBlend {blendInTime=0.1, blendOutTime=0.5}
    boneToModify {boneName=dyn_chain_1_f}
    rotation=<rot 10.852 3.699 1.272>
   }
   Output Pose(Result=Component To Local(ComponentPose=@n8)) {layerGroup=None}
  ]}
 EventGraph {nodes [
   Event BlueprintUpdateAnimation()
   > n1 = "Gravity Override State Function (Parameter Asset) (Nativized)"(
    CharacterPartAnimInstance=Self-Reference()
    GameTimeInSeconds=GetGameTimeInSeconds()
    Parameters=/BRCosmetics/Characters/Player/Male/Medium/Bodies/M_MED_GnatGala/Meshes/M_MED_GnatGala_Parameters
   )
   > Set Gravity Override(Gravity Override=@n1)
   n2 = Get bIsMoving2D()
   > Set CrouchingNotMoving(CrouchingNotMoving:bool=AND Boolean(A=Get bIsCrouching(), B=NOT Boolean(A=@n2))) -> (Output_Get:bool)
   PrintString!(execute, InString="To String (Name)"(InName=@n1.OutCurrentState), bPrintToScreen, bPrintToLog, TextColor, Duration, Key) -> (then)
  ]}
]
variables [
 CrouchingNotMoving {type=boolean}
 Gravity Override {type=Vector}
]
