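# Physics asset of cape colliders for the LentilAnise medium female player body, covering torso, skirt and hip gun.
# Capsules wrap pelvis, spine, clavicles, head and feet; arm, thigh, calf and neck bodies have no shapes.
# Skirt chains hang from spine_02 and spine_03, and a belt-to-gun chain from pelvis, each linking deform to dyn bones.
# All constraints leave swing and twist Free, so their angular limit values are inactive.
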PhysicsAsset /BRCosmetics/Characters/Player/Female/Medium/Bodies/F_MED_LentilAnise/Meshes/F_MED_LentilAnise_CapeColliders
bodies {
 calf_l []
 calf_r []
 clavicle_l [{shapeType=Capsule, length=6.5623, radius=7.5194, center=<4.5156 -1.1117 0.4235>, rotation=<rot -77.5566 177.3915 7.7314>}]
 clavicle_r [{shapeType=Capsule, length=6.5623, radius=7.5194, center=<-4.5156 1.1117 -0.4235>, rotation=<rot -77.5566 177.3915 -172.2686>}]
 deform_belt_gun_r [{shapeType=Capsule, length=1, radius=1, shapeName=deform_belt_gun_r_capsule}]
 deform_gun [{shapeType=Capsule, length=1, radius=1, shapeName=deform_gun_capsule}]
 deform_skirt_b_l [{shapeType=Capsule, length=1, radius=1, shapeName=deform_skirt_b_l_capsule}]
 deform_skirt_b_r [{shapeType=Capsule, length=1, radius=1, shapeName=deform_skirt_b_r_capsule, center=<0 -2.2724 0>}]
 deform_skirt_bk_l [{shapeType=Capsule, length=1, radius=1, shapeName=deform_skirt_bk_l_capsule}]
 deform_skirt_bk_r [{shapeType=Capsule, length=1, radius=1, shapeName=deform_skirt_bk_r_capsule}]
 deform_skirt_m_r [{shapeType=Capsule, length=1, radius=1, shapeName=deform_skirt_m_r_capsule}]
 dyn_belt_gun_r [{shapeType=Capsule, length=1, radius=1, shapeName=dyn_belt_gun_r_capsule}]
 dyn_gun [{shapeType=Capsule, length=11.9722, radius=8.9987, shapeName=dyn_gun_capsule, center=<8.7983 -1.2021 3.0473>, rotation=<rot 73.992 0 0>}]
 dyn_gun_1 [{shapeType=Capsule, length=1, radius=1, shapeName=dyn_gun_1_capsule}]
 dyn_skirt_b_1_l [{shapeType=Capsule, length=10.1162, radius=10.8546, shapeName=dyn_skirt_b_1_l_capsule, center=<5.9828 10.2446 0>, rotation=<rot 90 0 0>}]
 dyn_skirt_b_1_r [{shapeType=Capsule, length=10.1162, radius=10.8546, shapeName=dyn_skirt_b_1_r_capsule, center=<5.9828 7.9722 0>, rotation=<rot 90 0 0>}]
 dyn_skirt_b_2_l [{shapeType=Capsule, length=10.1162, radius=10.8546, shapeName=dyn_skirt_b_2_l_capsule, center=<5.9828 7.3372 0>, rotation=<rot 90 0 0>}]
 dyn_skirt_b_2_r [{shapeType=Capsule, length=10.1162, radius=10.8546, shapeName=dyn_skirt_b_2_r_capsule, center=<5.9828 5.0648 0>, rotation=<rot 90 0 0>}]
 dyn_skirt_b_l [{shapeType=Capsule, length=10.1162, radius=10.8546, shapeName=dyn_skirt_b_l_capsule, center=<5.9828 10.2446 0>, rotation=<rot 90 0 0>}]
 dyn_skirt_b_r [{shapeType=Capsule, length=10.1162, radius=10.8546, shapeName=dyn_skirt_b_r_capsule, center=<5.9828 7.9722 0>, rotation=<rot 90 0 0>}]
 dyn_skirt_bk_1_l [{shapeType=Capsule, length=10.1162, radius=10.8546, shapeName=dyn_skirt_bk_1_l_capsule, center=<5.9828 10.2446 0>, rotation=<rot 90 0 0>}]
 dyn_skirt_bk_1_r [{shapeType=Capsule, length=10.1162, radius=10.8546, shapeName=dyn_skirt_bk_1_r_capsule, center=<5.9828 10.2446 0>, rotation=<rot 90 0 0>}]
 dyn_skirt_bk_2_l [{shapeType=Capsule, length=10.1162, radius=10.8546, shapeName=dyn_skirt_bk_2_l_capsule, center=<5.9828 7.3372 0>, rotation=<rot 90 0 0>}]
 dyn_skirt_bk_2_r [{shapeType=Capsule, length=10.1162, radius=10.8546, shapeName=dyn_skirt_bk_2_r_capsule, center=<5.9828 7.3372 0>, rotation=<rot 90 0 0>}]
 dyn_skirt_bk_l [{shapeType=Capsule, length=10.1162, radius=10.8546, shapeName=dyn_skirt_bk_l_capsule, center=<5.9828 10.2446 0>, rotation=<rot 90 0 0>}]
 dyn_skirt_bk_r [{shapeType=Capsule, length=10.1162, radius=10.8546, shapeName=dyn_skirt_bk_r_capsule, center=<5.9828 10.2446 0>, rotation=<rot 90 0 0>}]
 dyn_skirt_m_1_r [{shapeType=Capsule, length=11.0701, radius=9.4553, shapeName=dyn_skirt_m_1_r_capsule, center=<10.4427 5.0977 0>, rotation=<rot 90 0 0>}]
 dyn_skirt_m_2_r [{shapeType=Capsule, length=11.0701, radius=9.4553, shapeName=dyn_skirt_m_2_r_capsule, center=<10.4427 5.0977 0>, rotation=<rot 90 0 0>}]
 dyn_skirt_m_r [{shapeType=Capsule, length=11.0701, radius=9.4553, shapeName=dyn_skirt_m_r_capsule, center=<10.4427 5.0977 0>, rotation=<rot 90 0 0>}]
 foot_l [{shapeType=Capsule, length=11.0765, radius=8.0344, center=<-4.1811 6.0027 -0.338>, rotation=<rot 0.0388 1.4701 -89.3982>}]
 foot_r [{shapeType=Capsule, length=11.0765, radius=8.0344, center=<4.1811 -6.0027 0.338>, rotation=<rot 0.0388 1.4701 90.6018>}]
 head [{shapeType=Capsule, length=0, radius=18, center=<9.0697 0.0746 0>, rotation=<rot 0 -90 0>}]
 lowerarm_l []
 lowerarm_r []
 neck_02 []
 pelvis [{shapeType=Capsule, length=9.6042, radius=11.1441, center=<-2.4471 -2.5289 -0.1981>}]
 spine_01 [{shapeType=Capsule, length=4.3888, radius=14.3638, center=<-8.9021 -0.8973 0>}]
 spine_02 [
  {shapeType=Capsule, length=7.7057, radius=9.7317, center=<-2.6885 -1.2641 0>}
  {shapeType=Capsule, length=20.6875, radius=9.3125, center=<-1.8317 1.9362 0>, rotation=<rot 90 0 0>}
 ]
 spine_03 [
  {shapeType=Capsule, length=7.3351, radius=10.2877, center=<0.0574 0.4416 0>}
  {shapeType=Capsule, length=20.7188, radius=9.2812, center=<-1.2163 0.549 0>, rotation=<rot 90 180 -175.0732>}
 ]
 spine_04 [
  {shapeType=Capsule, length=9.0076, radius=10.3916, center=<3.1525 -0.6779 0>}
  {shapeType=Capsule, length=10.0312, radius=9.2812, center=<-1.2326 1.308 0>, rotation=<rot 90 0 0>}
 ]
 spine_05 [{shapeType=Capsule, length=9.678, radius=12.3571, center=<0.6969 1.91 0>}]
 thigh_l []
 thigh_r []
 upperarm_l []
 upperarm_r []
}
constraints [
 {
  bone1Name=deform_skirt_b_l
  bone2Name=spine_02
  swing1LimitDegrees=45
  swing1Motion=Free
  swing2LimitDegrees=45
  swing2Motion=Free
  twistLimitDegrees=45
  twistMotion=Free
 }
 {
  bone1Name=dyn_skirt_b_l
  bone2Name=deform_skirt_b_l
  swing1LimitDegrees=45
  swing1Motion=Free
  swing2LimitDegrees=45
  swing2Motion=Free
  twistLimitDegrees=45
  twistMotion=Free
 }
 {
  bone1Name=dyn_skirt_b_1_l
  bone2Name=dyn_skirt_b_l
  swing1LimitDegrees=45
  swing1Motion=Free
  swing2LimitDegrees=45
  swing2Motion=Free
  twistLimitDegrees=45
  twistMotion=Free
 }
 {
  bone1Name=dyn_skirt_b_2_l
  bone2Name=dyn_skirt_b_1_l
  swing1LimitDegrees=45
  swing1Motion=Free
  swing2LimitDegrees=45
  swing2Motion=Free
  twistLimitDegrees=45
  twistMotion=Free
 }
 {
  bone1Name=deform_skirt_b_r
  bone2Name=spine_02
  swing1LimitDegrees=45
  swing1Motion=Free
  swing2LimitDegrees=45
  swing2Motion=Free
  twistLimitDegrees=45
  twistMotion=Free
 }
 {
  bone1Name=dyn_skirt_b_r
  bone2Name=deform_skirt_b_r
  swing1LimitDegrees=45
  swing1Motion=Free
  swing2LimitDegrees=45
  swing2Motion=Free
  twistLimitDegrees=45
  twistMotion=Free
 }
 {
  bone1Name=dyn_skirt_b_1_r
  bone2Name=dyn_skirt_b_r
  swing1LimitDegrees=45
  swing1Motion=Free
  swing2LimitDegrees=45
  swing2Motion=Free
  twistLimitDegrees=45
  twistMotion=Free
 }
 {
  bone1Name=dyn_skirt_b_2_r
  bone2Name=dyn_skirt_b_1_r
  swing1LimitDegrees=45
  swing1Motion=Free
  swing2LimitDegrees=45
  swing2Motion=Free
  twistLimitDegrees=45
  twistMotion=Free
 }
 {
  bone1Name=deform_skirt_bk_l
  bone2Name=spine_02
  swing1LimitDegrees=45
  swing1Motion=Free
  swing2LimitDegrees=45
  swing2Motion=Free
  twistLimitDegrees=45
  twistMotion=Free
 }
 {
  bone1Name=dyn_skirt_bk_l
  bone2Name=deform_skirt_bk_l
  swing1LimitDegrees=45
  swing1Motion=Free
  swing2LimitDegrees=45
  swing2Motion=Free
  twistLimitDegrees=45
  twistMotion=Free
 }
 {
  bone1Name=dyn_skirt_bk_1_l
  bone2Name=dyn_skirt_bk_l
  swing1LimitDegrees=45
  swing1Motion=Free
  swing2LimitDegrees=45
  swing2Motion=Free
  twistLimitDegrees=45
  twistMotion=Free
 }
 {
  bone1Name=dyn_skirt_bk_2_l
  bone2Name=dyn_skirt_bk_1_l
  swing1LimitDegrees=45
  swing1Motion=Free
  swing2LimitDegrees=45
  swing2Motion=Free
  twistLimitDegrees=45
  twistMotion=Free
 }
 {
  bone1Name=deform_skirt_bk_r
  bone2Name=spine_02
  swing1LimitDegrees=45
  swing1Motion=Free
  swing2LimitDegrees=45
  swing2Motion=Free
  twistLimitDegrees=45
  twistMotion=Free
 }
 {
  bone1Name=dyn_skirt_bk_r
  bone2Name=deform_skirt_bk_r
  swing1LimitDegrees=45
  swing1Motion=Free
  swing2LimitDegrees=45
  swing2Motion=Free
  twistLimitDegrees=45
  twistMotion=Free
 }
 {
  bone1Name=dyn_skirt_bk_1_r
  bone2Name=dyn_skirt_bk_r
  swing1LimitDegrees=45
  swing1Motion=Free
  swing2LimitDegrees=45
  swing2Motion=Free
  twistLimitDegrees=45
  twistMotion=Free
 }
 {
  bone1Name=dyn_skirt_bk_2_r
  bone2Name=dyn_skirt_bk_1_r
  swing1LimitDegrees=45
  swing1Motion=Free
  swing2LimitDegrees=45
  swing2Motion=Free
  twistLimitDegrees=45
  twistMotion=Free
 }
 {
  bone1Name=deform_belt_gun_r
  bone2Name=pelvis
  swing1LimitDegrees=45
  swing1Motion=Free
  swing2LimitDegrees=45
  swing2Motion=Free
  twistLimitDegrees=45
  twistMotion=Free
 }
 {
  bone1Name=dyn_belt_gun_r
  bone2Name=deform_belt_gun_r
  swing1LimitDegrees=45
  swing1Motion=Free
  swing2LimitDegrees=45
  swing2Motion=Free
  twistLimitDegrees=45
  twistMotion=Free
 }
 {
  bone1Name=deform_gun
  bone2Name=dyn_belt_gun_r
  swing1LimitDegrees=45
  swing1Motion=Free
  swing2LimitDegrees=45
  swing2Motion=Free
  twistLimitDegrees=45
  twistMotion=Free
 }
 {
  bone1Name=dyn_gun
  bone2Name=deform_gun
  swing1LimitDegrees=45
  swing1Motion=Free
  swing2LimitDegrees=45
  swing2Motion=Free
  twistLimitDegrees=45
  twistMotion=Free
 }
 {
  bone1Name=dyn_gun_1
  bone2Name=dyn_gun
  swing1LimitDegrees=45
  swing1Motion=Free
  swing2LimitDegrees=45
  swing2Motion=Free
  twistLimitDegrees=45
  twistMotion=Free
 }
 {
  bone1Name=deform_skirt_m_r
  bone2Name=spine_03
  swing1LimitDegrees=45
  swing1Motion=Free
  swing2LimitDegrees=45
  swing2Motion=Free
  twistLimitDegrees=45
  twistMotion=Free
 }
 {
  bone1Name=dyn_skirt_m_r
  bone2Name=deform_skirt_m_r
  swing1LimitDegrees=45
  swing1Motion=Free
  swing2LimitDegrees=45
  swing2Motion=Free
  twistLimitDegrees=45
  twistMotion=Free
 }
 {
  bone1Name=dyn_skirt_m_1_r
  bone2Name=dyn_skirt_m_r
  swing1LimitDegrees=45
  swing1Motion=Free
  swing2LimitDegrees=45
  swing2Motion=Free
  twistLimitDegrees=45
  twistMotion=Free
 }
 {
  bone1Name=dyn_skirt_m_2_r
  bone2Name=dyn_skirt_m_1_r
  swing1LimitDegrees=45
  swing1Motion=Free
  swing2LimitDegrees=45
  swing2Motion=Free
  twistLimitDegrees=45
  twistMotion=Free
 }
]
