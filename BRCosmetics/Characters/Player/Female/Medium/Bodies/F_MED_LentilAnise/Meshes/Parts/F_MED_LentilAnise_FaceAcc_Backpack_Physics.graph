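# Physics asset for the backpack variant of the LentilAnise female face accessory, setting up five dynamic hair chains under the head.
# Head, upper-spine and left-shoulder/arm capsules form the body shapes, including a large offset spine_05 capsule, likely covering the backpack.
# Tiny deform_hair roots lead into angle-limited dyn_hair chains.
# Tighter cross-constraints tie neighboring chains together, so the hair moves as a connected sheet.
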PhysicsAsset /BRCosmetics/Characters/Player/Female/Medium/Bodies/F_MED_LentilAnise/Meshes/Parts/F_MED_LentilAnise_FaceAcc_Backpack_Physics
bodies {
 clavicle_l [{shapeType=Capsule, length=16.6506, radius=7.4635, shapeName=clavicle_l_capsule, center=<8.2708 0 0>, rotation=<rot 90 0 0>}]
 deform_hair_b [{shapeType=Capsule, length=0.505, radius=0.505, shapeName=deform_hair_b_capsule, rotation=<rot 0 -90 0>}]
 deform_hair_bk [{shapeType=Capsule, length=0.505, radius=0.505, shapeName=deform_hair_bk_capsule, rotation=<rot 0 -90 0>}]
 deform_hair_f [{shapeType=Capsule, length=0.505, radius=0.505, shapeName=deform_hair_f_capsule, rotation=<rot 0 -90 0>}]
 deform_hair_m [{shapeType=Capsule, length=0.505, radius=0.505, shapeName=deform_hair_m_capsule, rotation=<rot 0 -90 0>}]
 deform_hair_m_f [{shapeType=Capsule, length=0.505, radius=0.505, shapeName=deform_hair_m_f_capsule, rotation=<rot 0 -90 0>}]
 dyn_hair_1_b [{shapeType=Capsule, length=9.057, radius=5.943, shapeName=dyn_hair_1_b_capsule, center=<4.8817 -1.0816 0>}]
 dyn_hair_1_bk [{shapeType=Capsule, length=9.057, radius=5.943, shapeName=dyn_hair_1_bk_capsule, center=<4.8817 -1.0816 0>}]
 dyn_hair_1_f [{shapeType=Capsule, length=15.2069, radius=4.186, shapeName=dyn_hair_1_f_capsule, center=<7.1815 -2.4155 0>, rotation=<rot 90 0 0>}]
 dyn_hair_1_m [{shapeType=Capsule, length=9.057, radius=5.943, shapeName=dyn_hair_1_m_capsule, center=<2.5744 -1.0816 0>}]
 dyn_hair_1_m_f [{shapeType=Capsule, length=9.057, radius=5.943, shapeName=dyn_hair_1_m_f_capsule, center=<2.6738 -1.0816 0>}]
 dyn_hair_2_f [{shapeType=Capsule, length=10.0769, radius=4.9231, shapeName=dyn_hair_2_f_capsule, center=<6.1888 -0.5402 0>}]
 dyn_hair_b [{shapeType=Capsule, length=15.2069, radius=4.186, shapeName=dyn_hair_b_capsule, center=<7.1815 -2.4155 0>, rotation=<rot 90 0 0>}]
 dyn_hair_bk [{shapeType=Capsule, length=15.2069, radius=4.186, shapeName=dyn_hair_bk_capsule, center=<7.1815 -2.4155 0>, rotation=<rot 90 0 0>}]
 dyn_hair_f [{shapeType=Capsule, length=12.9901, radius=5.3078, shapeName=dyn_hair_f_capsule, center=<6.3706 -4.006 0>, rotation=<rot 90 0 0>}]
 dyn_hair_m [{shapeType=Capsule, length=15.2069, radius=4.186, shapeName=dyn_hair_m_capsule, center=<7.1815 -2.4155 0>, rotation=<rot 90 0 0>}]
 dyn_hair_m_f [{shapeType=Capsule, length=15.2069, radius=4.186, shapeName=dyn_hair_m_f_capsule, center=<7.1815 -2.4155 0>, rotation=<rot 90 0 0>}]
 head [{shapeType=Capsule, length=12.7635, radius=10.0043, shapeName=head_capsule, center=<8.3271 3.0393 0>, rotation=<rot 90 0 0>}]
 spine_04 [{shapeType=Capsule, length=5.4144, radius=12.2728, shapeName=spine_04_capsule, center=<2.1071 2.193 0>, rotation=<rot 0 -90 0>}]
 spine_05 [
  {shapeType=Capsule, length=12.1513, radius=11.4201, shapeName=spine_05_capsule, center=<2.9286 2.2095 0>}
  {shapeType=Capsule, length=19.4787, radius=24.0425, shapeName=spine_05_capsule, center=<-4.2561 -47.3917 0>, rotation=<rot 90 0 0>}
 ]
 upperarm_l [{shapeType=Capsule, length=19.4368, radius=7.6129, shapeName=upperarm_l_capsule, center=<10.7461 0 0>, rotation=<rot 90 0 0>}]
}
constraints [
 {
  bone1Name=spine_05
  bone2Name=spine_04
  swing1LimitDegrees=45
  swing1Motion=Free
  swing2LimitDegrees=45
  swing2Motion=Free
  twistLimitDegrees=45
  twistMotion=Free
 }
 {
  bone1Name=clavicle_l
  bone2Name=spine_05
  swing1LimitDegrees=45
  swing1Motion=Free
  swing2LimitDegrees=45
  swing2Motion=Free
  twistLimitDegrees=45
  twistMotion=Free
 }
 {
  bone1Name=upperarm_l
  bone2Name=clavicle_l
  swing1LimitDegrees=45
  swing1Motion=Free
  swing2LimitDegrees=45
  swing2Motion=Free
  twistLimitDegrees=45
  twistMotion=Free
 }
 {
  bone1Name=head
  bone2Name=spine_05
  swing1LimitDegrees=45
  swing1Motion=Free
  swing2LimitDegrees=45
  swing2Motion=Free
  twistLimitDegrees=45
  twistMotion=Free
 }
 {bone1Name=dyn_hair_b, bone2Name=deform_hair_b, swing1LimitDegrees=30, swing2LimitDegrees=15, twistLimitDegrees=5}
 {bone1Name=dyn_hair_1_b, bone2Name=dyn_hair_b, swing1LimitDegrees=30, swing2LimitDegrees=15, twistLimitDegrees=5}
 {bone1Name=deform_hair_b, bone2Name=head, swing1LimitDegrees=30, swing2LimitDegrees=15, twistLimitDegrees=5}
 {
  bone1Name=deform_hair_bk
  bone2Name=head
  swing1LimitDegrees=45
  swing1Motion=Free
  swing2LimitDegrees=45
  swing2Motion=Free
  twistLimitDegrees=45
  twistMotion=Free
 }
 {bone1Name=dyn_hair_bk, bone2Name=deform_hair_bk, swing1LimitDegrees=30, swing2LimitDegrees=15, twistLimitDegrees=5}
 {bone1Name=dyn_hair_1_bk, bone2Name=dyn_hair_bk, swing1LimitDegrees=30, swing2LimitDegrees=15, twistLimitDegrees=5}
 {
  bone1Name=deform_hair_f
  bone2Name=head
  swing1LimitDegrees=45
  swing1Motion=Free
  swing2LimitDegrees=45
  swing2Motion=Free
  twistLimitDegrees=45
  twistMotion=Free
 }
 {bone1Name=dyn_hair_f, bone2Name=deform_hair_f, swing1LimitDegrees=10, swing2LimitDegrees=15, twistLimitDegrees=5}
 {bone1Name=dyn_hair_1_f, bone2Name=dyn_hair_f, swing1LimitDegrees=30, swing2LimitDegrees=30, twistLimitDegrees=5}
 {bone1Name=dyn_hair_2_f, bone2Name=dyn_hair_1_f, swing1LimitDegrees=30, swing2LimitDegrees=15, twistLimitDegrees=5}
 {
  bone1Name=deform_hair_m
  bone2Name=head
  swing1LimitDegrees=45
  swing1Motion=Free
  swing2LimitDegrees=45
  swing2Motion=Free
  twistLimitDegrees=45
  twistMotion=Free
 }
 {bone1Name=dyn_hair_m, bone2Name=deform_hair_m, swing1LimitDegrees=30, swing2LimitDegrees=30, twistLimitDegrees=5}
 {bone1Name=dyn_hair_1_m, bone2Name=dyn_hair_m, swing1LimitDegrees=30, swing2LimitDegrees=15, twistLimitDegrees=5}
 {
  bone1Name=deform_hair_m_f
  bone2Name=head
  swing1LimitDegrees=45
  swing1Motion=Free
  swing2LimitDegrees=45
  swing2Motion=Free
  twistLimitDegrees=45
  twistMotion=Free
 }
 {bone1Name=dyn_hair_m_f, bone2Name=deform_hair_m_f, swing1LimitDegrees=30, swing2LimitDegrees=30, twistLimitDegrees=5}
 {bone1Name=dyn_hair_1_m_f, bone2Name=dyn_hair_m_f, swing1LimitDegrees=30, swing2LimitDegrees=15, twistLimitDegrees=5}
 {bone1Name=dyn_hair_2_f, bone2Name=dyn_hair_1_m_f, swing1LimitDegrees=10, swing2LimitDegrees=10, twistLimitDegrees=10}
 {bone1Name=dyn_hair_1_m_f, bone2Name=dyn_hair_1_m, swing1LimitDegrees=15, swing2LimitDegrees=15, twistLimitDegrees=15}
 {bone1Name=dyn_hair_1_m, bone2Name=dyn_hair_1_b, swing1LimitDegrees=15, swing2LimitDegrees=15, twistLimitDegrees=15}
 {bone1Name=dyn_hair_1_b, bone2Name=dyn_hair_1_bk, swing1LimitDegrees=15, swing2LimitDegrees=15, twistLimitDegrees=15}
 {bone1Name=dyn_hair_f, bone2Name=dyn_hair_m_f, swing1LimitDegrees=5, swing2LimitDegrees=5, twistLimitDegrees=5}
 {bone1Name=dyn_hair_1_f, bone2Name=dyn_hair_m_f, swing1LimitDegrees=10, swing2LimitDegrees=10, twistLimitDegrees=10}
 {bone1Name=dyn_hair_m_f, bone2Name=dyn_hair_m, swing1LimitDegrees=10, swing2LimitDegrees=10, twistLimitDegrees=10}
]
