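# Control Rig for the LentilAnise female face accessory mesh, applying pose-driven corrective offsets to hair deform bones.
# A spherical pose reader on the left upper arm, relative to spine_05, weights transform modifications on front hair bones.
# When the WearingCape variable is true, a second pose reader on the head weights further offsets on middle hair bones.
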ControlRigBlueprint /BRCosmetics/Characters/Player/Female/Medium/Bodies/F_MED_LentilAnise/Meshes/Parts/F_MED_LentilAnise_FaceAcc_CtrlRig
graphs [
 RigVMFunctionLibrary {links [], nodes []}
 RigVMModel {nodes [
   RigUnit_BeginExecution = &RigUnit_BeginExecution()
   n1 = &RigUnit_SphericalPoseReader(
    ActiveRegionScaleFactors=(PositiveWidth=1,NegativeWidth=1,PositiveHeight=1,NegativeHeight=1)
    ActiveRegionSize=0.2
    Debug=(bDrawDebug=true,bDraw2D=false,DebugScale=25,DebugSegments=20,DebugThickness=0.25)
    DriverAxis=(X=0,Y=1,Z=0)
    DriverItem=(Type=Bone,Name="upperarm_l")
    FalloffRegionScaleFactors=(PositiveWidth=1,NegativeWidth=1,PositiveHeight=1,NegativeHeight=1)
    FalloffSize=0.5
    OptionalParentItem=(Type=Bone,Name="spine_05")
    RotationOffset=(X=0,Y=90,Z=0)
    ExecutePin=@RigUnit_BeginExecution.ExecutePin
   )
   > n2 = &RigUnit_ModifyTransforms(
    ItemToModify="((Item=(Type=Bone,Name=\"deform_hair_f\"),Transform=(Rotation=(X=0,Y=0.2164,Z=0,W=0.9763),Translation=(X=0,Y=0,Z=0),Scale3D=(X=1,Y=1,Z=1))),(Item=(Type=Bone,Name=\"deform_hair_m_f\"),Transform=(Rotation=(X=0,Y=0.1305,Z=0,W=0.9914),Translation=(X=0,Y=1,Z=-4),Sc ...(truncated)"
    WeightMinimum=0
    Weight=@n1.OutputParam
   )
   RigVMFunction_ControlFlowBranch = &RigVMFunction_ControlFlowBranch(ExecuteContext=@n2.ExecutePin, Condition=RigVMVariableNode() {variable=WearingCape}.Value)
   n3 = &RigUnit_SphericalPoseReader(
    ActiveRegionScaleFactors=(PositiveWidth=1,NegativeWidth=1,PositiveHeight=1,NegativeHeight=1)
    ActiveRegionSize=0.2
    Debug=(bDrawDebug=true,bDraw2D=false,DebugScale=25,DebugSegments=20,DebugThickness=0.25)
    DriverAxis=(X=1,Y=0,Z=0)
    DriverItem=(Type=Bone,Name="head")
    FalloffRegionScaleFactors=(PositiveWidth=1,NegativeWidth=1,PositiveHeight=1,NegativeHeight=1)
    FalloffSize=0.5
    OptionalParentItem=(Type=Bone,Name="spine_05")
    RotationOffset=(X=-90,Y=0,Z=-30)
    ExecutePin=@RigVMFunction_ControlFlowBranch.True
   )
   > &RigUnit_ModifyTransforms(
    ItemToModify="((Item=(Type=Bone,Name=\"deform_hair_m\"),Transform=(Rotation=(X=0,Y=0,Z=0,W=1),Translation=(X=-5,Y=0,Z=0),Scale3D=(X=1,Y=1,Z=1))),(Item=(Type=Bone,Name=\"deform_hair_m_f\"),Transform=(Rotation=(X=0,Y=0,Z=0,W=1),Translation=(X=-5,Y=0,Z=0),Scale3D=(X=1,Y=1,Z=1) ...(truncated)"
    WeightMinimum=0
    Weight=@n3.OutputParam
   )
  ]}
]
hierarchy [
 root {loc=<0 0 0>, rot=<0 0 0 1>}
 pelvis {parent=root, loc=<0 -0.5301 98.0571>, rot=<0.0127 0.707 -0.0127 -0.707>}
 spine_01 {parent=pelvis, loc=<6.1208 -0.9328 0>, rot=<0 0 0.034 -0.9994>}
 spine_02 {parent=spine_01, loc=<6.5997 -0.2095 0>, rot=<0 0 0.0238 -0.9997>}
 spine_03 {parent=spine_02, loc=<6.4914 -0.1237 0>, rot=<0 0 0.0149 -0.9999>}
 spine_04 {parent=spine_03, loc=<6.9598 0.0794 0>, rot=<0 0 -0.0132 -0.9999>}
 spine_05 {parent=spine_04, loc=<6.9987 0 0>, rot=<0 0 -0.0015 -1>}
 neck_01 {parent=spine_05, loc=<11.3138 0.1529 0>, rot=<0 0 -0.2003 -0.9797>}
 neck_02 {parent=neck_01, loc=<4.5627 0 0>, rot=<0 0 0.0504 -0.9987>}
 head {parent=neck_02, loc=<5.3914 0 0>, rot=<0 0 0.115 -0.9934>}
 deform_hair_bk {parent=head, loc=<15.5139 -12.3589 0>, rot=<0 -1 0 0>}
 dyn_hair_bk {parent=deform_hair_bk, loc=<0 0 0>, rot=<0 0 0 1>}
 dyn_hair_1_bk {parent=dyn_hair_bk, loc=<15.2002 0 0>, rot=<0 0 0 1>}
 deform_hair_b {parent=head, loc=<15.4963 -10.152 -6.6211>, rot=<0.0299 -0.9749 -0.2204 -0.0068>}
 dyn_hair_b {parent=deform_hair_b, loc=<0 0 0>, rot=<0 0 0 1>}
 dyn_hair_1_b {parent=dyn_hair_b, loc=<15.2002 0 0>, rot=<0 0 0 1>}
 deform_hair_m {parent=head, loc=<15.4511 -6.8561 -11.1567>, rot=<0.033 -0.8494 -0.5261 -0.0236>}
 dyn_hair_m {parent=deform_hair_m, loc=<0 0 0>, rot=<0 0 0 1>}
 dyn_hair_1_m {parent=dyn_hair_m, loc=<15.2002 0 0>, rot=<0 0 0 1>}
 deform_hair_f {parent=head, loc=<25.1477 7.8144 -2.5656>, rot=<-0.4244 0.5953 0.6335 0.2532>}
 dyn_hair_f {parent=deform_hair_f, loc=<0 0 0>, rot=<0 0 0 1>}
 dyn_hair_1_f {parent=dyn_hair_f, loc=<14.0312 2.3552 -0.0067>, rot=<0.0284 0.1988 -0.4685 -0.8604>}
 dyn_hair_2_f {parent=dyn_hair_1_f, loc=<15.2002 0 0>, rot=<0 0 0 1>}
 deform_hair_m_f {parent=head, loc=<15.4287 -0.1495 -13.4365>, rot=<0.0286 -0.7228 -0.6899 -0.0269>}
 dyn_hair_m_f {parent=deform_hair_m_f, loc=<0 0 0>, rot=<0 0 0 1>}
 dyn_hair_1_m_f {parent=dyn_hair_m_f, loc=<15.2002 0 0>, rot=<0 0 0 1>}
 faceAttach {parent=head, loc=<-3.8533 -0.7383 0>, rot=<0.0029 -0.7071 -0.0029 -0.7071>}
 C_brow_mid {parent=faceAttach, loc=<-0.0311 12.8901 21.0725>, rot=<-0.0646 0 0 -0.9979>}
 R_brow_mid {parent=faceAttach, loc=<-4.9726 11.7507 21.7109>, rot=<0.006 -0.0365 -0.1553 -0.9872>}
 R_brow_outer {parent=faceAttach, loc=<-7.0012 10.6175 21.4341>, rot=<-0.0686 0.0135 -0.4809 -0.874>}
 clavicle_l {parent=spine_05, loc=<5.4831 -0.1053 -3.2946>, rot=<0.0465 -0.7289 -0.009 -0.6829>}
 upperarm_l {parent=clavicle_l, loc=<9.8658 -1.2136 -0.4105>, rot=<0.0164 -0.3245 -0.0269 -0.9454>}
 upperarm_twist_01_l {parent=upperarm_l, loc=<7.2135 -0.1134 0>, rot=<0 0 0 1>}
 deform_upperarm_bk_base_l {parent=upperarm_twist_01_l, loc=<-6.4053 -4.3746 0.314>, rot=<0 0 0 -1>}
 deform_upperarm_bk_l {parent=deform_upperarm_bk_base_l, loc=<0 0 0>, rot=<0 0 0 1>}
 deform_upperarm_fr_base_l {parent=upperarm_twist_01_l, loc=<-5.6312 5.092 0.3354>, rot=<0 0 0 -1>}
 deform_upperarm_fr_l {parent=deform_upperarm_fr_base_l, loc=<0 0 0>, rot=<0 0 0 1>}
 deform_upperarm_up_base_l {parent=upperarm_twist_01_l, loc=<-8.891 0.4825 3.9517>, rot=<0 0 0 -1>}
 deform_upperarm_up_l {parent=deform_upperarm_up_base_l, loc=<0 0 0>, rot=<0 0 0 1>}
 deform_collar_l {parent=clavicle_l, loc=<4.4496 7.9068 5.6378>, rot=<-0.544 -0.0501 0.6737 0.4977>}
 dyn_collar_l {parent=deform_collar_l, loc=<0 0 0>, rot=<0 0 0 1>}
]
variables [WearingCape {type=bool, default=False}]
