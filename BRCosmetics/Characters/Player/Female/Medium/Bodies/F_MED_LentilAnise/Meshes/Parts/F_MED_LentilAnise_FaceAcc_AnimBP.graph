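# Animation Blueprint for the LentilAnise female character's face accessory, adding secondary physics over poses copied from the parent and body meshes.
# The layered pose feeds a cape-aware Control Rig, three head-space RigidBody nodes using base, backpack and cape physics assets, then a front-hair bone offset.
# The event graph caches the body mesh, sets backpack and cape flags, and computes gravity override settings from another skin's (IcicleCane_Pitch) parameter asset.
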AnimBlueprint /BRCosmetics/Characters/Player/Female/Medium/Bodies/F_MED_LentilAnise/Meshes/Parts/F_MED_LentilAnise_FaceAcc_AnimBP
target_skeleton=&/F_MED_LentilAnise_FaceAcc_Skeleton
graphs [
 AnimGraph {nodes [
   RigidBody!(ComponentPose, Alpha, OverrideWorldGravity:Vector="0,0,-980") -> (Pose) {
    bClampLinearTranslationLimitToRefPose=true
    bOverrideWorldGravity=true
    overridePhysicsAsset=&/F_MED_LentilAnise_FaceAcc_Physics
    simulationSpace=BaseBoneSpace
    alphaBoolBlend {blendOption=Linear}
    baseBoneRef {boneName=head}
    overrideWorldGravity {z=-980}
    simSpaceSettings {
     maxAngularAcceleration=1500
     maxAngularVelocity=5000
     maxLinearAcceleration=300
     maxLinearVelocity=5000
     worldAlpha=0.6
     externalLinearDragV=<0.2 0.2 0.2>
    }
   }
   n1 = Layered blend per bone(
    BasePose=Copy Pose From Mesh() {bUseAttachedParent=true}
    BlendPoses_0=Copy Pose From Mesh(SourceMeshComponent=Get Body_Mesh()) {bUseAttachedParent=true}
   ) {layerSetup [{branchFilters [{blendDepth=1, boneName=clavicle_l}, {blendDepth=1, boneName=clavicle_r}, {blendDepth=1, boneName=upperarm_l}]}]}
   n2 = Control Rig!(Source=@n1, Alpha, WearingCape:bool=Get Is Backbling A Cape() -> (Is Backbling A Cape:bool)) -> (Pose) {controlRigAssetReference {blueprintRigClass=&/F_MED_LentilAnise_FaceAcc_CtrlRig.F_MED_LentilAnise_FaceAcc_CtrlRig_C}}
   n3 = RigidBody!(
    ComponentPose=Local To Component(LocalPose=@n2)
    bAlphaBoolEnabled:bool=false
    OverrideWorldGravity="0,0,-980"
    SimSpaceSettings:Sim Space Settings Structure=(WorldAlpha=0.6,VelocityScaleZ=1,DampingAlpha=1,MaxLinearVelocity=5000,MaxAngularVelocity=5000,MaxLinearAcceleration=300,MaxAngularAcceleration=1500,ExternalLinearDragV=(X=0.2,Y=0.2,Z=0.2),ExternalLinearVelocity=(X=0,Y=0,Z=0),ExternalAngularVelocity=(X=0,Y=0,Z=0))
   ) -> (Pose) {
    alphaInputType=Bool
    bClampLinearTranslationLimitToRefPose=true
    bOverrideWorldGravity=true
    overridePhysicsAsset=&F_MED_LentilAnise_FaceAcc_Physics
    simulationSpace=BaseBoneSpace
    alphaScaleBias {bias=1, scale=-1}
    baseBoneRef {boneName=head}
    overrideWorldGravity {z=-980}
    simSpaceSettings {
     maxAngularAcceleration=1500
     maxAngularVelocity=5000
     maxLinearAcceleration=300
     maxLinearVelocity=5000
     worldAlpha=0.6
     externalLinearDragV=<0.2 0.2 0.2>
    }
   }
   n4 = RigidBody(
    ComponentPose=@n3
    OverrideWorldGravity="0,0,-980"
    SimSpaceSettings=(WorldAlpha=0.6,VelocityScaleZ=1,DampingAlpha=1,MaxLinearVelocity=5000,MaxAngularVelocity=5000,MaxLinearAcceleration=300,MaxAngularAcceleration=1500,ExternalLinearDragV=(X=0.2,Y=0.2,Z=0.2),ExternalLinearVelocity=(X=0,Y=0,Z=0),ExternalAngularVelocity=(X=0,Y=0,Z=0))
   ) {
    alphaInputType=Bool
    bClampLinearTranslationLimitToRefPose=true
    bOverrideWorldGravity=true
    overridePhysicsAsset=&/F_MED_LentilAnise_FaceAcc_Backpack_Physics
    simulationSpace=BaseBoneSpace
    alphaScaleBias {scale=2}
    baseBoneRef {boneName=head}
    overrideWorldGravity {z=-980}
    simSpaceSettings {
     maxAngularAcceleration=1500
     maxAngularVelocity=5000
     maxLinearAcceleration=300
     maxLinearVelocity=5000
     worldAlpha=0.6
     externalLinearDragV=<0.2 0.2 0.2>
    }
   }
   n5 = RigidBody(
    ComponentPose=@n4
    OverrideWorldGravity="0,0,-980"
    SimSpaceSettings=(WorldAlpha=0.6,VelocityScaleZ=1,DampingAlpha=1,MaxLinearVelocity=5000,MaxAngularVelocity=5000,MaxLinearAcceleration=300,MaxAngularAcceleration=1500,ExternalLinearDragV=(X=0.2,Y=0.2,Z=0.2),ExternalLinearVelocity=(X=0,Y=0,Z=0),ExternalAngularVelocity=(X=0,Y=0,Z=0))
   ) {
    alphaInputType=Bool
    bClampLinearTranslationLimitToRefPose=true
    bOverrideWorldGravity=true
    overridePhysicsAsset=&/F_MED_LentilAnise_FaceAcc_Cape_Physics
    simulationSpace=BaseBoneSpace
    alphaScaleBias {bias=-1, scale=2}
    baseBoneRef {boneName=head}
    overrideWorldGravity {z=-980}
    simSpaceSettings {
     maxAngularAcceleration=1500
     maxAngularVelocity=5000
     maxLinearAcceleration=300
     maxLinearVelocity=5000
     worldAlpha=0.6
     externalLinearDragV=<0.2 0.2 0.2>
    }
   }
   n6 = "Transform (Modify) Bone - Bone: deform_hair_f"(ComponentPose=@n5, Alpha=1, Translation="1.6,0,0", Rotation="0,0,0", Scale="1,1,1") {
    rotationMode=BMM_Additive
    rotationSpace=BCS_BoneSpace
    translationMode=BMM_Additive
    translationSpace=BCS_BoneSpace
    boneToModify {boneName=deform_hair_f}
    translation {x=1.6}
   }
   Output Pose(Result=Component To Local(ComponentPose=@n6)) {layerGroup=None}
  ]}
 EventGraph {nodes [
   Event BlueprintUpdateAnimation()
   > n1 = "Gravity Override State Function (Parameter Asset) (Nativized)"(
    CharacterPartAnimInstance=Self-Reference()
    GameTimeInSeconds=GetGameTimeInSeconds()
    Parameters=/BRCosmetics/Characters/Player/Female/Medium/Bodies/F_MED_IcicleCane_Pitch/Meshes/Parts/F_MED_IcicleCane_Pitch_FaceAcc_Parameters
   )
   > Set Gravity Override(Gravity Override=@n1)
   > Set RBAN Settings Out(RBAN Settings Out=@n1.OutRigidBodyAnimNodeParameters)
   n2 = Event OnCharacterCustomizationLoaded()
   > Set Body_Mesh(Body_Mesh=Get Array Item with EFort Custom Part Type(Array=@n2.CustomMeshParts))
   n3 = CheckForBackbling()
   > Set NotWearingBackpack(NotWearingBackpack:bool=NOT Boolean(A=@n3)) -> (Output_Get:bool)
   > n4 = Branch(Condition=@n3)
   > n5 = Branch(Condition=@n3.bIsBackblingACape)
   > n6 = Set Is Backbling A Cape(Is Backbling A Cape:bool=true) -> (Output_Get:bool)
   Set Is Wearing Backpack(execute=[@n7.then, @n6.then], Is Wearing Backpack:bool=false) -> (Output_Get:bool)
   Set Is Wearing Backpack(execute=@n5.else, Is Wearing Backpack=true)
   > Set Is Backbling A Cape(Is Backbling A Cape=false)
   n7 = Set Is Backbling A Cape(execute=@n4.else, Is Backbling A Cape=false)
  ]}
]
variables [
 Body_Mesh {type=SkeletalMeshComponent}
 Gravity Override {type=Vector}
 Is Backbling A Cape {type=boolean}
 Is Wearing Backpack {type=boolean}
 NotWearingBackpack {type=boolean}
 RBAN Settings Out {type=FortRigidBodyAnimNodeParameters}
]
